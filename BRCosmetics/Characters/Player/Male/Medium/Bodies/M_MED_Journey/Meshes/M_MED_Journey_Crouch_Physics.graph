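# Physics asset for the crouch variant of the male medium Journey body cosmetic, simulating its skirt around the legs.
# Eight bone chains of capsule bodies hang from small master bodies around the skirt, with two thigh capsules that have no constraints.
# Tight swing limits, with twist mostly locked at zero degrees or kept narrow, keep each chain stiff.
# Free-angle cross constraints tie some back and mid chains together sideways.
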PhysicsAsset /BRCosmetics/Characters/Player/Male/Medium/Bodies/M_MED_Journey/Meshes/M_MED_Journey_Crouch_Physics
bodies {
 dyn_skirt_bl_01 [{shapeType=Capsule, length=8, radius=3, shapeName=dyn_skirt_bl_01_capsule, center=<2 0 0>}]
 dyn_skirt_bl_02 [{shapeType=Capsule, length=8, radius=5.4, shapeName=dyn_skirt_bl_02_capsule, center=<2 0 0>}]
 dyn_skirt_bl_03 [{shapeType=Capsule, length=8, radius=5.4, shapeName=dyn_skirt_bl_03_capsule, center=<2 0 0>}]
 dyn_skirt_bl_04 [{shapeType=Capsule, length=8, radius=5.4, shapeName=dyn_skirt_bl_04_capsule, center=<2 0 0>}]
 dyn_skirt_bl_05 [{shapeType=Capsule, length=8, radius=5.4, shapeName=dyn_skirt_bl_05_capsule, center=<2 0 0>}]
 dyn_skirt_bl_master [{shapeType=Capsule, length=0.505, radius=0.505, shapeName=dyn_skirt_bl_master_capsule}]
 dyn_skirt_br_01 [{shapeType=Capsule, length=8, radius=3, shapeName=dyn_skirt_br_01_capsule, center=<2 0 0>}]
 dyn_skirt_br_02 [{shapeType=Capsule, length=8, radius=5.4, shapeName=dyn_skirt_br_02_capsule, center=<2 0 0>}]
 dyn_skirt_br_03 [{shapeType=Capsule, length=8, radius=5.4, shapeName=dyn_skirt_br_03_capsule, center=<2 0 0>}]
 dyn_skirt_br_04 [{shapeType=Capsule, length=8, radius=5.4, shapeName=dyn_skirt_br_04_capsule, center=<2 0 0>, rotation=<rot 0 0.0007 0>}]
 dyn_skirt_br_05 [{shapeType=Capsule, length=8, radius=5.4, shapeName=dyn_skirt_br_05_capsule, center=<2 0 0>}]
 dyn_skirt_br_master [{shapeType=Capsule, length=0.505, radius=0.505, shapeName=dyn_skirt_br_master_capsule}]
 dyn_skirt_fl_01 [{shapeType=Capsule, length=8, radius=3, shapeName=dyn_skirt_fl_01_capsule, center=<2.5 0 0>}]
 dyn_skirt_fl_02 [{shapeType=Capsule, length=8, radius=3, shapeName=dyn_skirt_fl_02_capsule, center=<2.5 0 0>}]
 dyn_skirt_fl_03 [{shapeType=Capsule, length=6.1855, radius=4.8145, shapeName=dyn_skirt_fl_03_capsule, center=<2.5 0 0>}]
 dyn_skirt_fl_04 [{shapeType=Capsule, length=6.1855, radius=4.8145, shapeName=dyn_skirt_fl_04_capsule, center=<2.5 0 0>}]
 dyn_skirt_fl_master [{shapeType=Capsule, length=0.505, radius=0.505, shapeName=dyn_skirt_fl_master_capsule, rotation=<rot 0 -90 0>}]
 dyn_skirt_fr_01 [{shapeType=Capsule, length=8, radius=3, shapeName=dyn_skirt_fr_01_capsule, center=<2.5 0 0>}]
 dyn_skirt_fr_02 [{shapeType=Capsule, length=10.341, radius=3.4824, shapeName=dyn_skirt_fr_02_capsule, center=<2.5 0 0>}]
 dyn_skirt_fr_03 [{shapeType=Capsule, length=10.1278, radius=3.7387, shapeName=dyn_skirt_fr_03_capsule, center=<1.8806 0.0001 0>}]
 dyn_skirt_fr_04 [{shapeType=Capsule, length=7.6266, radius=3.7387, shapeName=dyn_skirt_fr_04_capsule}]
 dyn_skirt_fr_master [{shapeType=Capsule, length=0.505, radius=0.505, shapeName=dyn_skirt_fr_master_capsule, rotation=<rot 0 -90 0>}]
 dyn_skirt_l_01 [{shapeType=Capsule, length=8, radius=3, shapeName=dyn_skirt_l_01_capsule, center=<2 0 0>}]
 dyn_skirt_l_02 [{shapeType=Capsule, length=8, radius=3, shapeName=dyn_skirt_l_02_capsule, center=<2 0 0>}]
 dyn_skirt_l_03 [{shapeType=Capsule, length=7.0251, radius=3.9749, shapeName=dyn_skirt_l_03_capsule, center=<2 0 0>}]
 dyn_skirt_l_04 [{shapeType=Capsule, length=7.0251, radius=3.9749, shapeName=dyn_skirt_l_04_capsule, center=<2 0 0>}]
 dyn_skirt_l_master [{shapeType=Capsule, length=0.505, radius=0.505, shapeName=dyn_skirt_l_master_capsule}]
 dyn_skirt_midl_01 [{shapeType=Capsule, length=8, radius=3, shapeName=dyn_skirt_midl_01_capsule, center=<2 0.3314 0.3195>}]
 dyn_skirt_midl_02 [{shapeType=Capsule, length=8, radius=3, shapeName=dyn_skirt_midl_02_capsule, center=<2 0.3314 0.3195>}]
 dyn_skirt_midl_03 [{shapeType=Capsule, length=8, radius=3, shapeName=dyn_skirt_midl_03_capsule, center=<2 0.3314 0.3195>}]
 dyn_skirt_midl_04 [{shapeType=Capsule, length=8, radius=3, shapeName=dyn_skirt_midl_04_capsule, center=<2 0.3314 0.3195>}]
 dyn_skirt_midl_master [{shapeType=Capsule, length=0.505, radius=0.505, shapeName=dyn_skirt_midl_master_capsule}]
 dyn_skirt_midr_01 [{shapeType=Capsule, length=8, radius=3, shapeName=dyn_skirt_midr_01_capsule, center=<2 0 0>}]
 dyn_skirt_midr_02 [{shapeType=Capsule, length=8, radius=3, shapeName=dyn_skirt_midr_02_capsule, center=<2 0 0>}]
 dyn_skirt_midr_03 [{shapeType=Capsule, length=8, radius=3, shapeName=dyn_skirt_midr_03_capsule, center=<2 0 0>}]
 dyn_skirt_midr_04 [{shapeType=Capsule, length=8, radius=3, shapeName=dyn_skirt_midr_04_capsule, center=<2 0 0>}]
 dyn_skirt_midr_master [{shapeType=Capsule, length=0.505, radius=0.505, shapeName=dyn_skirt_midr_master_capsule}]
 dyn_skirt_r_01 [{shapeType=Capsule, length=8, radius=3, shapeName=dyn_skirt_r_01_capsule, center=<2.5 0.66 0>}]
 dyn_skirt_r_02 [{shapeType=Capsule, length=11.5014, radius=3.285, shapeName=dyn_skirt_r_02_capsule, center=<2.5 0.66 0>}]
 dyn_skirt_r_03 [{shapeType=Capsule, length=14.1476, radius=3.7555, shapeName=dyn_skirt_r_03_capsule, center=<2.5 0.66 4.6153>}]
 dyn_skirt_r_04 [{shapeType=Capsule, length=10.2381, radius=3.7555, shapeName=dyn_skirt_r_04_capsule, center=<2.5 0.66 5.388>}]
 dyn_skirt_r_master [{shapeType=Capsule, length=0.505, radius=0.505, shapeName=dyn_skirt_r_master_capsule}]
 thigh_l [{shapeType=Capsule, length=27.4334, radius=7.4536, shapeName=thigh_l_capsule, center=<-25 0 0>, rotation=<rot 90 0 0>}]
 thigh_r [{shapeType=Capsule, length=25.2592, radius=9.473, shapeName=thigh_r_capsule, center=<25 0 0>, rotation=<rot 90 0 0>}]
}
constraints [
 {bone1Name=dyn_skirt_r_01, bone2Name=dyn_skirt_r_master, swing1LimitDegrees=10, swing2LimitDegrees=5, twistLimitDegrees=0}
 {bone1Name=dyn_skirt_r_02, bone2Name=dyn_skirt_r_01, swing1LimitDegrees=10, swing2LimitDegrees=5, twistLimitDegrees=0}
 {bone1Name=dyn_skirt_r_03, bone2Name=dyn_skirt_r_02, swing1LimitDegrees=10, swing2LimitDegrees=5, twistLimitDegrees=0}
 {bone1Name=dyn_skirt_r_04, bone2Name=dyn_skirt_r_03, swing1LimitDegrees=10, swing2LimitDegrees=5, twistLimitDegrees=0}
 {bone1Name=dyn_skirt_midr_01, bone2Name=dyn_skirt_midr_master, swing1LimitDegrees=5, swing2LimitDegrees=5, twistLimitDegrees=0}
 {bone1Name=dyn_skirt_midr_02, bone2Name=dyn_skirt_midr_01, swing1LimitDegrees=5, swing2LimitDegrees=5, twistLimitDegrees=0}
 {bone1Name=dyn_skirt_midr_03, bone2Name=dyn_skirt_midr_02, swing1LimitDegrees=5, swing2LimitDegrees=5, twistLimitDegrees=0}
 {bone1Name=dyn_skirt_midr_04, bone2Name=dyn_skirt_midr_03, swing1LimitDegrees=5, swing2LimitDegrees=5, twistLimitDegrees=0}
 {bone1Name=dyn_skirt_br_01, bone2Name=dyn_skirt_br_master, swing1LimitDegrees=5, swing2LimitDegrees=5, twistLimitDegrees=0}
 {bone1Name=dyn_skirt_br_02, bone2Name=dyn_skirt_br_01, swing1LimitDegrees=5, swing2LimitDegrees=5, twistLimitDegrees=0}
 {bone1Name=dyn_skirt_br_03, bone2Name=dyn_skirt_br_02, swing1LimitDegrees=5, swing2LimitDegrees=5, twistLimitDegrees=0}
 {bone1Name=dyn_skirt_br_04, bone2Name=dyn_skirt_br_03, swing1LimitDegrees=5, swing2LimitDegrees=5, twistLimitDegrees=0}
 {bone1Name=dyn_skirt_br_05, bone2Name=dyn_skirt_br_04, swing1LimitDegrees=5, swing2LimitDegrees=5, twistLimitDegrees=0}
 {bone1Name=dyn_skirt_bl_01, bone2Name=dyn_skirt_bl_master, swing1LimitDegrees=5, swing2LimitDegrees=5, twistLimitDegrees=0}
 {bone1Name=dyn_skirt_bl_02, bone2Name=dyn_skirt_bl_01, swing1LimitDegrees=5, swing2LimitDegrees=5, twistLimitDegrees=0}
 {bone1Name=dyn_skirt_bl_03, bone2Name=dyn_skirt_bl_02, swing1LimitDegrees=5, swing2LimitDegrees=5, twistLimitDegrees=0}
 {bone1Name=dyn_skirt_bl_04, bone2Name=dyn_skirt_bl_03, swing1LimitDegrees=5, swing2LimitDegrees=5, twistLimitDegrees=0}
 {bone1Name=dyn_skirt_bl_05, bone2Name=dyn_skirt_bl_04, swing1LimitDegrees=5, swing2LimitDegrees=5, twistLimitDegrees=0}
 {bone1Name=dyn_skirt_l_01, bone2Name=dyn_skirt_l_master, swing1LimitDegrees=5, swing2LimitDegrees=5, twistLimitDegrees=0}
 {bone1Name=dyn_skirt_l_02, bone2Name=dyn_skirt_l_01, swing1LimitDegrees=5, swing2LimitDegrees=5, twistLimitDegrees=0}
 {bone1Name=dyn_skirt_l_03, bone2Name=dyn_skirt_l_02, swing1LimitDegrees=5, swing2LimitDegrees=5, twistLimitDegrees=0}
 {bone1Name=dyn_skirt_l_04, bone2Name=dyn_skirt_l_03, swing1LimitDegrees=5, swing2LimitDegrees=5, twistLimitDegrees=0}
 {bone1Name=dyn_skirt_midl_01, bone2Name=dyn_skirt_midl_master, swing1LimitDegrees=2, swing2LimitDegrees=5, twistLimitDegrees=0}
 {bone1Name=dyn_skirt_midl_02, bone2Name=dyn_skirt_midl_01, swing1LimitDegrees=2, swing2LimitDegrees=5, twistLimitDegrees=0}
 {bone1Name=dyn_skirt_midl_03, bone2Name=dyn_skirt_midl_02, swing1LimitDegrees=2, swing2LimitDegrees=5, twistLimitDegrees=0}
 {bone1Name=dyn_skirt_midl_04, bone2Name=dyn_skirt_midl_03, swing1LimitDegrees=2, swing2LimitDegrees=5, twistLimitDegrees=0}
 {
  bone1Name=dyn_skirt_midr_01
  bone2Name=dyn_skirt_br_01
  swing1LimitDegrees=45
  swing1Motion=Free
  swing2LimitDegrees=45
  swing2Motion=Free
  twistLimitDegrees=45
  twistMotion=Free
 }
 {
  bone1Name=dyn_skirt_midr_02
  bone2Name=dyn_skirt_br_02
  swing1LimitDegrees=45
  swing1Motion=Free
  swing2LimitDegrees=45
  swing2Motion=Free
  twistLimitDegrees=45
  twistMotion=Free
 }
 {
  bone1Name=dyn_skirt_midr_03
  bone2Name=dyn_skirt_br_03
  swing1LimitDegrees=45
  swing1Motion=Free
  swing2LimitDegrees=45
  swing2Motion=Free
  twistLimitDegrees=45
  twistMotion=Free
 }
 {
  bone1Name=dyn_skirt_midr_04
  bone2Name=dyn_skirt_br_04
  swing1LimitDegrees=45
  swing1Motion=Free
  swing2LimitDegrees=45
  swing2Motion=Free
  twistLimitDegrees=45
  twistMotion=Free
 }
 {
  bone1Name=dyn_skirt_bl_02
  bone2Name=dyn_skirt_midl_02
  swing1LimitDegrees=45
  swing1Motion=Free
  swing2LimitDegrees=45
  swing2Motion=Free
  twistLimitDegrees=45
  twistMotion=Free
 }
 {
  bone1Name=dyn_skirt_bl_03
  bone2Name=dyn_skirt_midl_03
  swing1LimitDegrees=45
  swing1Motion=Free
  swing2LimitDegrees=45
  swing2Motion=Free
  twistLimitDegrees=45
  twistMotion=Free
 }
 {bone1Name=dyn_skirt_fl_02, bone2Name=dyn_skirt_fl_01, swing1LimitDegrees=5, swing2LimitDegrees=5, twistLimitDegrees=5}
 {bone1Name=dyn_skirt_fl_03, bone2Name=dyn_skirt_fl_02, swing1LimitDegrees=5, swing2LimitDegrees=5, twistLimitDegrees=5}
 {bone1Name=dyn_skirt_fl_04, bone2Name=dyn_skirt_fl_03, swing1LimitDegrees=5, swing2LimitDegrees=5, twistLimitDegrees=5}
 {bone1Name=dyn_skirt_fr_02, bone2Name=dyn_skirt_fr_01, swing1LimitDegrees=5, swing2LimitDegrees=5, twistLimitDegrees=5}
 {bone1Name=dyn_skirt_fr_03, bone2Name=dyn_skirt_fr_02, swing1LimitDegrees=7, swing2LimitDegrees=5, twistLimitDegrees=5}
 {bone1Name=dyn_skirt_fr_04, bone2Name=dyn_skirt_fr_03, swing1LimitDegrees=7, swing2LimitDegrees=5, twistLimitDegrees=0}
 {bone1Name=dyn_skirt_fl_01, bone2Name=dyn_skirt_fl_master, swing1LimitDegrees=5, swing2LimitDegrees=5, twistLimitDegrees=5}
 {bone1Name=dyn_skirt_fr_01, bone2Name=dyn_skirt_fr_master, swing1LimitDegrees=5, swing2LimitDegrees=5, twistLimitDegrees=5}
]
