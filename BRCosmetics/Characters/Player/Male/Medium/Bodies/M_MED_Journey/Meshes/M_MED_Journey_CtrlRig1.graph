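# Control Rig for the M_MED_Journey male medium body that adjusts accessory bones on the character skeleton.
# The executed chain applies fixed rotation offsets to both tassle master bones, then parent-constrains each pickaxe_attach bone to a weighted blend of its forearm twist bones.
# Two further chains rotate the skirt master bones but are not connected to Begin Execution.
# A CrouchingAlpha variable is declared but unused by the graph.
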ControlRigBlueprint /BRCosmetics/Characters/Player/Male/Medium/Bodies/M_MED_Journey/Meshes/M_MED_Journey_CtrlRig1
graphs [
 RigVMFunctionLibrary {links [], nodes []}
 RigVMModel {nodes [
   &RigUnit_ModifyTransforms(
    ExecutePin=()
    ItemToModify="((Item=(Type=Bone,Name=\"None\"),Transform=(Rotation=(X=0,Y=0,Z=0,W=1),Translation=(X=0,Y=0,Z=0),Scale3D=(X=1,Y=1,Z=1))),(Item=(Type=Bone,Name=\"dyn_skirt_bl_master\"),Transform=(Rotation=(X=0,Y=0,Z=0.0872,W=0.9962),Translation=(X=0,Y=0,Z=0),Scale3D=(X=1,Y=1,Z ...(truncated)"
    Weight=1
   )
   > &RigUnit_ModifyTransforms(
    ItemToModify="((Item=(Type=Bone,Name=\"None\"),Transform=(Rotation=(X=0,Y=0,Z=0,W=1),Translation=(X=0,Y=0,Z=0),Scale3D=(X=1,Y=1,Z=1))),(Item=(Type=Bone,Name=\"dyn_skirt_br_master\"),Transform=(Rotation=(X=0,Y=0,Z=0.0872,W=0.9962),Translation=(X=0,Y=0,Z=0),Scale3D=(X=1,Y=1,Z ...(truncated)"
    Weight=1
   )
   > &RigUnit_ModifyTransforms(
    ItemToModify="((Item=(Type=Bone,Name=\"None\"),Transform=(Rotation=(X=0,Y=0,Z=0,W=1),Translation=(X=0,Y=0,Z=0),Scale3D=(X=1,Y=1,Z=1))),(Item=(Type=Bone,Name=\"dyn_skirt_l_master\"),Transform=(Rotation=(X=0,Y=0,Z=0.0872,W=0.9962),Translation=(X=0,Y=0,Z=0),Scale3D=(X=1,Y=1,Z= ...(truncated)"
    Weight=1
   )
   > &RigUnit_ModifyTransforms(
    ItemToModify="((Item=(Type=Bone,Name=\"None\"),Transform=(Rotation=(X=0,Y=0,Z=0,W=1),Translation=(X=0,Y=0,Z=0),Scale3D=(X=1,Y=1,Z=1))),(Item=(Type=Bone,Name=\"dyn_skirt_r_master\"),Transform=(Rotation=(X=0,Y=0,Z=0.0872,W=0.9962),Translation=(X=0,Y=0,Z=0),Scale3D=(X=1,Y=1,Z= ...(truncated)"
    Weight=1
   )
   > &RigUnit_ModifyTransforms(
    ItemToModify="((Item=(Type=Bone,Name=\"None\"),Transform=(Rotation=(X=0,Y=0,Z=0,W=1),Translation=(X=0,Y=0,Z=0),Scale3D=(X=1,Y=1,Z=1))),(Item=(Type=Bone,Name=\"dyn_skirt_midl_master\"),Transform=(Rotation=(X=0,Y=0,Z=0.0872,W=0.9962),Translation=(X=0,Y=0,Z=0),Scale3D=(X=1,Y=1 ...(truncated)"
    Weight=1
   )
   > &RigUnit_ModifyTransforms(
    ItemToModify="((Item=(Type=Bone,Name=\"None\"),Transform=(Rotation=(X=0,Y=0,Z=0,W=1),Translation=(X=0,Y=0,Z=0),Scale3D=(X=1,Y=1,Z=1))),(Item=(Type=Bone,Name=\"dyn_skirt_midr_master\"),Transform=(Rotation=(X=0,Y=0,Z=0.0872,W=0.9962),Translation=(X=0,Y=0,Z=0),Scale3D=(X=1,Y=1 ...(truncated)"
    Weight=1
   )
   &RigUnit_ModifyTransforms(
    ExecutePin=()
    ItemToModify="((Item=(Type=Bone,Name=\"None\"),Transform=(Rotation=(X=0,Y=0,Z=0,W=1),Translation=(X=0,Y=0,Z=0),Scale3D=(X=1,Y=1,Z=1))),(Item=(Type=Bone,Name=\"dyn_skirt_fl_master\"),Transform=(Rotation=(X=0,Y=0,Z=-0.1305,W=0.9914),Translation=(X=0,Y=0,Z=0),Scale3D=(X=1,Y=1, ...(truncated)"
    Weight=1
   )
   > &RigUnit_ModifyTransforms(
    ItemToModify="((Item=(Type=Bone,Name=\"None\"),Transform=(Rotation=(X=0,Y=0,Z=0,W=1),Translation=(X=0,Y=0,Z=0),Scale3D=(X=1,Y=1,Z=1))),(Item=(Type=Bone,Name=\"dyn_skirt_fr_master\"),Transform=(Rotation=(X=0,Y=0,Z=-0.1305,W=0.9914),Translation=(X=0,Y=0,Z=0),Scale3D=(X=1,Y=1, ...(truncated)"
    Weight=1
   )
   RigUnit_BeginExecution = &RigUnit_BeginExecution()
   &RigUnit_ModifyTransforms(
    ItemToModify="((Item=(Type=Bone,Name=\"None\"),Transform=(Rotation=(X=0,Y=0,Z=0,W=1),Translation=(X=0,Y=0,Z=0),Scale3D=(X=1,Y=1,Z=1))),(Item=(Type=Bone,Name=\"tassle_l_master\"),Transform=(Rotation=(X=0,Y=0,Z=0.1305,W=0.9914),Translation=(X=0,Y=0,Z=0),Scale3D=(X=1,Y=1,Z=1)) ...(truncated)"
    Weight=1
    ExecutePin=@RigUnit_BeginExecution.ExecutePin
   )
   > &RigUnit_ModifyTransforms(
    ItemToModify="((Item=(Type=Bone,Name=\"None\"),Transform=(Rotation=(X=0,Y=0,Z=0,W=1),Translation=(X=0,Y=0,Z=0),Scale3D=(X=1,Y=1,Z=1))),(Item=(Type=Bone,Name=\"tassle_r_master\"),Transform=(Rotation=(X=0,Y=0,Z=0.1305,W=0.9914),Translation=(X=0,Y=0,Z=0),Scale3D=(X=1,Y=1,Z=1)) ...(truncated)"
    Weight=1
   )
   > &RigUnit_ParentConstraint(
    AdvancedSettings=(RotationOrderForFilter=XZY)
    Child=(Type=Bone,Name="pickaxe_attach_l")
    Parents=((Item=(Type=Bone,Name="lowerarm_twist_01_l"),Weight=0.25),(Item=(Type=Bone,Name="lowerarm_twist_02_l"),Weight=0.75))
    Weight=1
   )
   > &RigUnit_ParentConstraint(
    AdvancedSettings=(RotationOrderForFilter=XZY)
    Child=(Type=Bone,Name="pickaxe_attach_r")
    Parents=((Item=(Type=Bone,Name="lowerarm_twist_01_r"),Weight=0.25),(Item=(Type=Bone,Name="lowerarm_twist_02_r"),Weight=0.75))
    Weight=1
   )
  ]}
]
hierarchy [
 root {loc=<0 0 0>, rot=<0 0 0 1>}
 attach {parent=root, loc=<0 0 0>, rot=<0 0 0 1>}
 pelvis {parent=root, loc=<0 -0.7071 93.1191>, rot=<0.0127 0.707 -0.0127 -0.707>}
 spine_01 {parent=pelvis, loc=<6.7494 -0.6388 0>, rot=<0 0 0.034 -0.9994>}
 spine_02 {parent=spine_01, loc=<7.2518 -0.035 0>, rot=<0 0 0.0238 -0.9997>}
 spine_03 {parent=spine_02, loc=<6.0104 0 0>, rot=<0 0 0.0149 -0.9999>}
 spine_04 {parent=spine_03, loc=<8.1024 0 0>, rot=<0 0 -0.0132 -0.9999>}
 spine_05 {parent=spine_04, loc=<10.0146 0 0>, rot=<0 0 -0.0015 -1>}
 clavicle_l {parent=spine_05, loc=<4.3575 -0.1466 -2.3223>, rot=<0.0465 -0.7289 -0.009 -0.6829>}
 upperarm_l {parent=clavicle_l, loc=<15.5111 0 0>, rot=<0.0226 -0.3234 -0.0088 -0.9459>}
 lowerarm_l {parent=upperarm_l, loc=<26.6514 0 0>, rot=<0.0391 -0.0099 -0.2453 -0.9686>}
 hand_l {parent=lowerarm_l, loc=<22.3491 0 0>, rot=<-0.7081 -0.0507 0.0975 -0.6975>}
 index_metacarpal_l {parent=hand_l, loc=<1.6535 0.0183 -2.96>, rot=<0.0158 -0.0124 -0.0173 -0.9996>}
 index_01_l {parent=index_metacarpal_l, loc=<8.3008 0.0357 -0.0524>, rot=<-0.0044 0.0485 0.0924 -0.9945>}
 index_02_l {parent=index_01_l, loc=<3.3587 -0.3004 0.0452>, rot=<-0.002 0.0027 0.3225 -0.9466>}
 index_03_l {parent=index_02_l, loc=<2.5578 -0.123 -0.0252>, rot=<-0.002 0.0067 0.2558 -0.9667>}
 middle_metacarpal_l {parent=hand_l, loc=<2.1856 -0.1725 -0.5984>, rot=<0.006 0.0168 -0.0329 -0.9993>}
 middle_01_l {parent=middle_metacarpal_l, loc=<7.3275 -0.2447 0.0129>, rot=<-0.0208 0.0042 0.0858 -0.9961>}
 middle_02_l {parent=middle_01_l, loc=<4.0123 0.1401 0.1175>, rot=<-0.0218 0.0258 0.3512 -0.9357>}
 middle_03_l {parent=middle_02_l, loc=<2.6077 0.0534 0.1373>, rot=<0.0082 0.0188 0.1708 -0.9851>}
 pinky_metacarpal_l {parent=hand_l, loc=<1.7616 -0.9753 3.3814>, rot=<-0.0397 0.0814 -0.069 -0.9935>}
 pinky_01_l {parent=pinky_metacarpal_l, loc=<7.1709 -0.4831 -0.1094>, rot=<-0.0654 -0.0035 0.2242 -0.9723>}
 pinky_02_l {parent=pinky_01_l, loc=<2.8274 0.0893 -0.1043>, rot=<0.0037 0.0011 0.2854 -0.9584>}
 pinky_03_l {parent=pinky_02_l, loc=<2.2046 0.0307 -0.0078>, rot=<-0.0002 0.0098 0.1445 -0.9895>}
 ring_metacarpal_l {parent=hand_l, loc=<2.2074 -0.3625 1.3638>, rot=<-0.0284 0.0722 -0.0458 -0.9959>}
 ring_01_l {parent=ring_metacarpal_l, loc=<7.0295 -0.6197 -0.1446>, rot=<-0.0489 -0.0252 0.1892 -0.9804>}
 ring_02_l {parent=ring_01_l, loc=<3.6174 0.4647 -0.1027>, rot=<0 0 0.2806 -0.9598>}
 ring_03_l {parent=ring_02_l, loc=<2.4152 0.114 0.0199>, rot=<0 0 0.1189 -0.9929>}
 thumb_01_l {parent=hand_l, loc=<1.3118 -0.3191 -2.459>, rot=<0.6546 -0.4042 0.0664 -0.6354>}
 thumb_02_l {parent=thumb_01_l, loc=<5.9336 -0.0448 0.0401>, rot=<-0.0037 0.0213 0.298 -0.9543>}
 thumb_03_l {parent=thumb_02_l, loc=<2.2223 0.3682 -0.0907>, rot=<0.0118 -0.013 0.0066 -0.9998>}
 weapon_l {parent=hand_l, loc=<-0.1974 -3.4111 0.3811>, rot=<0 0 0 1>}
 lowerarm_twist_01_l {parent=lowerarm_l, loc=<17.8549 0 0>, rot=<0 0 0 1>}
 lowerarm_twist_02_l {parent=lowerarm_l, loc=<9.6199 0 0>, rot=<0 0 0 1>}
 blade_l_base {parent=lowerarm_l, loc=<3.9082 1.0856 -5.3895>, rot=<-0.7276 0.0569 -0.0373 -0.6826>}
 blade_l_end {parent=lowerarm_l, loc=<8.8367 0.9275 -4.733>, rot=<-0.7276 0.0569 -0.0373 -0.6826>}
 pickaxe_attach_l {parent=lowerarm_l, loc=<3.9082 1.0856 -5.3895>, rot=<-0.7276 0.0569 -0.0373 -0.6826>}
 upperarm_twist_01_l {parent=upperarm_l, loc=<0.0041 0 0>, rot=<0 0 0 1>}
 upperarm_twist_02_l {parent=upperarm_l, loc=<13.7941 0 0>, rot=<0 0 0 1>}
 clavicle_r {parent=spine_05, loc=<4.3575 -0.1466 2.3223>, rot=<0.7289 0.0465 -0.6829 0.009>}
 upperarm_r {parent=clavicle_r, loc=<-15.5111 0 0>, rot=<0.0226 -0.3234 -0.0088 -0.9459>}
 lowerarm_r {parent=upperarm_r, loc=<-26.6515 0 0>, rot=<0.0391 -0.0099 -0.2453 -0.9686>}
 hand_r {parent=lowerarm_r, loc=<-22.3492 0 0>, rot=<-0.7081 -0.0507 0.0975 -0.6975>}
 index_metacarpal_r {parent=hand_r, loc=<-1.6534 -0.0181 2.9601>, rot=<0.0158 -0.0124 -0.0173 -0.9996>}
 index_01_r {parent=index_metacarpal_r, loc=<-8.3008 -0.0357 0.0524>, rot=<-0.0044 0.0485 0.0924 -0.9945>}
 index_02_r {parent=index_01_r, loc=<-3.3589 0.3003 -0.0453>, rot=<-0.002 0.0027 0.3225 -0.9466>}
 index_03_r {parent=index_02_r, loc=<-2.5578 0.1231 0.0251>, rot=<-0.002 0.0067 0.2558 -0.9667>}
 middle_metacarpal_r {parent=hand_r, loc=<-2.1856 0.1726 0.5984>, rot=<0.006 0.0168 -0.0329 -0.9993>}
 middle_01_r {parent=middle_metacarpal_r, loc=<-7.3273 0.2447 -0.0129>, rot=<-0.0208 0.0042 0.0858 -0.9961>}
 middle_02_r {parent=middle_01_r, loc=<-4.0142 -0.1411 -0.07>, rot=<-0.0217 0.0259 0.3512 -0.9357>}
 middle_03_r {parent=middle_02_r, loc=<-2.6074 -0.0534 -0.1312>, rot=<0.0083 0.0188 0.1708 -0.9851>}
 pinky_metacarpal_r {parent=hand_r, loc=<-1.7617 0.9753 -3.3814>, rot=<-0.0397 0.0814 -0.069 -0.9935>}
 pinky_01_r {parent=pinky_metacarpal_r, loc=<-7.1707 0.4832 0.1094>, rot=<-0.0654 -0.0035 0.2242 -0.9723>}
 pinky_02_r {parent=pinky_01_r, loc=<-2.8274 -0.0892 0.1043>, rot=<0.0037 0.0011 0.2854 -0.9584>}
 pinky_03_r {parent=pinky_02_r, loc=<-2.2045 -0.031 0.0078>, rot=<-0.0002 0.0098 0.1445 -0.9895>}
 ring_metacarpal_r {parent=hand_r, loc=<-2.2073 0.3627 -1.3638>, rot=<-0.0284 0.0722 -0.0458 -0.9959>}
 ring_01_r {parent=ring_metacarpal_r, loc=<-7.0297 0.6197 0.1446>, rot=<-0.0489 -0.0252 0.1892 -0.9804>}
 ring_02_r {parent=ring_01_r, loc=<-3.6171 -0.4648 0.1027>, rot=<0 0 0.2806 -0.9598>}
 ring_03_r {parent=ring_02_r, loc=<-2.4153 -0.1141 -0.0199>, rot=<0 0 0.1189 -0.9929>}
 thumb_01_r {parent=hand_r, loc=<-1.3119 0.3192 2.459>, rot=<0.6546 -0.4042 0.0664 -0.6354>}
 thumb_02_r {parent=thumb_01_r, loc=<-5.9335 0.0447 -0.0401>, rot=<-0.0037 0.0213 0.298 -0.9543>}
 thumb_03_r {parent=thumb_02_r, loc=<-2.2223 -0.3682 0.0907>, rot=<0.0118 -0.013 0.0066 -0.9998>}
 weapon_r {parent=hand_r, loc=<0.1975 3.4112 -0.3811>, rot=<0 0 0 -1>}
 lowerarm_twist_01_r {parent=lowerarm_r, loc=<-17.855 0 0>, rot=<0 0 0 1>}
 lowerarm_twist_02_r {parent=lowerarm_r, loc=<-9.6199 0 0>, rot=<0 0 0 1>}
 blade_r_base {parent=lowerarm_r, loc=<-3.8711 -1.1075 5.4036>, rot=<-0.0388 0.6631 0.7454 0.0559>}
 blade_r_end {parent=lowerarm_r, loc=<-9.8003 -0.9172 4.6137>, rot=<-0.0388 0.6631 0.7454 0.0559>}
 pickaxe_attach_r {parent=lowerarm_r, loc=<-3.8711 -1.1075 5.4036>, rot=<-0.0388 0.6631 0.7454 0.0559>}
 upperarm_twist_01_r {parent=upperarm_r, loc=<-0.0041 0 0>, rot=<0 0 0 1>}
 upperarm_twist_02_r {parent=upperarm_r, loc=<-13.7942 0 0>, rot=<0 0 0 1>}
 neck_01 {parent=spine_05, loc=<11.3138 0.1529 0>, rot=<0 0 -0.2003 -0.9797>}
 neck_02 {parent=neck_01, loc=<4.5627 0 0>, rot=<0 0 0.0504 -0.9987>}
 head {parent=neck_02, loc=<5.3914 0 0>, rot=<0 0 0.115 -0.9934>}
 coat_flap_br_master {parent=spine_05, loc=<5.6451 10.3338 6.2893>, rot=<0.4576 -0.1486 0.8353 0.266>}
 coat_flap_br_01 {parent=coat_flap_br_master, loc=<0 0 0>, rot=<0 0 0 1>}
 coat_flap_br_02 {parent=coat_flap_br_01, loc=<4.2001 0 0>, rot=<0.0295 0.0624 -0.015 -0.9975>}
 coat_flap_br_03 {parent=coat_flap_br_02, loc=<5.6354 0 0>, rot=<1 0 0 0>}
 coat_flap_tr_master {parent=spine_05, loc=<6.5768 10.5242 4.8804>, rot=<0.8378 -0.1524 0.429 0.3014>}
 coat_flap_tr_01 {parent=coat_flap_tr_master, loc=<0 0 0>, rot=<0 0 0 1>}
 coat_flap_tr_02 {parent=coat_flap_tr_01, loc=<3.1649 0 0>, rot=<0.0177 -0.0227 0.0369 -0.9989>}
 coat_flap_tr_03 {parent=coat_flap_tr_02, loc=<2.9775 0 0>, rot=<1 0 0 0>}
 coat_flap_tl_master {parent=spine_05, loc=<6.5768 10.5242 -4.8804>, rot=<0.8378 -0.1524 -0.429 -0.3014>}
 coat_flap_tl_01 {parent=coat_flap_tl_master, loc=<0 0 0>, rot=<0 0 0 1>}
 coat_flap_tl_02 {parent=coat_flap_tl_01, loc=<3.165 0 0>, rot=<-0.0177 0.0227 0.0369 -0.9989>}
 coat_flap_tl_03 {parent=coat_flap_tl_02, loc=<2.9774 0 0>, rot=<0 0 0 1>}
 coat_flap_bl_master {parent=spine_05, loc=<5.6446 10.3337 -6.2893>, rot=<-0.4576 0.1486 0.8353 0.266>}
 coat_flap_bl_01 {parent=coat_flap_bl_master, loc=<0 0 0>, rot=<0 0 0 1>}
 coat_flap_bl_02 {parent=coat_flap_bl_01, loc=<4.2001 0 0>, rot=<-0.0295 -0.0624 -0.0151 -0.9975>}
 coat_flap_bl_03 {parent=coat_flap_bl_02, loc=<5.6355 0 0>, rot=<0 0 0 1>}
 thigh_l {parent=pelvis, loc=<-2.5693 -0.0109 -7.7501>, rot=<0.0012 0.0572 0.0155 -0.9982>}
 calf_l {parent=thigh_l, loc=<-39.0484 0 0>, rot=<0.0006 -0.0161 -0.0391 -0.9991>}
 calf_twist_01_l {parent=calf_l, loc=<-35.813 -0.0967 0.0397>, rot=<0.0237 0.0095 -0.0035 -0.9997>}
 calf_twist_02_l {parent=calf_l, loc=<-18.8064 -0.0967 0.0397>, rot=<0.0237 0.0095 -0.0035 -0.9997>}
 foot_l {parent=calf_l, loc=<-42.561 0.1077 0.4292>, rot=<0.0418 -0.0455 0.0343 -0.9975>}
 ball_l {parent=foot_l, loc=<-6.1599 13.8264 0.6806>, rot=<0.0001 -0.0001 -0.7063 -0.7079>}
 thigh_twist_01_l {parent=thigh_l, loc=<-21.6303 0 0>, rot=<0.0013 0 0 -1>}
 thigh_r {parent=pelvis, loc=<-2.5691 -0.0109 7.7501>, rot=<0.0572 -0.0012 0.9982 0.0155>}
 calf_r {parent=thigh_r, loc=<39.0487 0 0>, rot=<0.0006 -0.0161 -0.0391 -0.9991>}
 calf_twist_01_r {parent=calf_r, loc=<35.8132 0.0967 -0.0397>, rot=<0.0237 0.0095 -0.0035 -0.9997>}
 calf_twist_02_r {parent=calf_r, loc=<18.8064 0.0967 -0.0397>, rot=<0.0237 0.0095 -0.0035 -0.9997>}
 foot_r {parent=calf_r, loc=<42.5612 -0.1077 -0.4292>, rot=<0.0418 -0.0455 0.0343 -0.9975>}
 ball_r {parent=foot_r, loc=<6.1598 -13.8264 -0.6806>, rot=<0.0001 -0.0001 -0.7063 -0.7079>}
 thigh_twist_01_r {parent=thigh_r, loc=<21.6303 0 0>, rot=<0.0013 0 0 -1>}
 dyn_skirt_r_master {parent=pelvis, loc=<4.3091 4.5627 12.8381>, rot=<-0.0801 -0.81 -0.5718 0.1031>}
 dyn_skirt_r_01 {parent=dyn_skirt_r_master, loc=<0 0 0>, rot=<0 0 0 1>}
 dyn_skirt_r_02 {parent=dyn_skirt_r_01, loc=<7.639 0 0>, rot=<0 0 0 1>}
 dyn_skirt_r_03 {parent=dyn_skirt_r_02, loc=<9.2509 0.0001 0>, rot=<0 0 0 1>}
 dyn_skirt_r_04 {parent=dyn_skirt_r_03, loc=<8.3994 0 0>, rot=<0 0 0 1>}
 dyn_skirt_midr_master {parent=pelvis, loc=<-1.3858 -7.3204 11.8153>, rot=<0.0939 0.3277 0.9383 -0.0587>}
 dyn_skirt_midr_01 {parent=dyn_skirt_midr_master, loc=<0 0 0>, rot=<0 0 0 1>}
 dyn_skirt_midr_02 {parent=dyn_skirt_midr_01, loc=<9.7843 0 0>, rot=<0 0 0 1>}
 dyn_skirt_midr_03 {parent=dyn_skirt_midr_02, loc=<11.1245 0 0>, rot=<0 0 0 -1>}
 dyn_skirt_midr_04 {parent=dyn_skirt_midr_03, loc=<10.1286 0 0>, rot=<0 0 0 1>}
 dyn_skirt_br_master {parent=pelvis, loc=<-2.8876 -12.2697 2.4639>, rot=<0.0578 0.0024 0.9975 -0.0395>}
 dyn_skirt_br_01 {parent=dyn_skirt_br_master, loc=<0 0 0>, rot=<0 0 0 1>}
 dyn_skirt_br_02 {parent=dyn_skirt_br_01, loc=<7.6389 0 0>, rot=<0 0 0 1>}
 dyn_skirt_br_03 {parent=dyn_skirt_br_02, loc=<10.0892 0 0>, rot=<0 0 0 1>}
 dyn_skirt_br_04 {parent=dyn_skirt_br_03, loc=<10.2247 0 -0.0001>, rot=<0 0 0 1>}
 dyn_skirt_br_05 {parent=dyn_skirt_br_04, loc=<8.5797 0 0>, rot=<0 0 0 1>}
 dyn_skirt_bl_master {parent=pelvis, loc=<-2.8876 -12.2697 -2.4639>, rot=<-0.0578 -0.0024 0.9975 -0.0395>}
 dyn_skirt_bl_01 {parent=dyn_skirt_bl_master, loc=<0 0 0>, rot=<0 0 0 1>}
 dyn_skirt_bl_02 {parent=dyn_skirt_bl_01, loc=<7.6389 0 0>, rot=<0 0 0 1>}
 dyn_skirt_bl_03 {parent=dyn_skirt_bl_02, loc=<10.0892 0 0>, rot=<0 0 0 1>}
 dyn_skirt_bl_04 {parent=dyn_skirt_bl_03, loc=<10.2247 0 0>, rot=<0 0 0 1>}
 dyn_skirt_bl_05 {parent=dyn_skirt_bl_04, loc=<8.5797 0 0>, rot=<0 0 0 1>}
 dyn_skirt_l_master {parent=pelvis, loc=<4.1956 4.4512 -12.625>, rot=<-0.087 -0.8112 0.5707 -0.0933>}
 dyn_skirt_l_01 {parent=dyn_skirt_l_master, loc=<0 0 0>, rot=<0 0 0 1>}
 dyn_skirt_l_02 {parent=dyn_skirt_l_01, loc=<7.6389 0 0>, rot=<0 0 0 1>}
 dyn_skirt_l_03 {parent=dyn_skirt_l_02, loc=<9.2509 0 0>, rot=<0 0 0 1>}
 dyn_skirt_l_04 {parent=dyn_skirt_l_03, loc=<8.3994 0 0>, rot=<0 0 0 1>}
 dyn_skirt_midl_master {parent=pelvis, loc=<-1.3857 -7.3204 -11.8153>, rot=<-0.0939 -0.3277 0.9383 -0.0587>}
 dyn_skirt_midl_01 {parent=dyn_skirt_midl_master, loc=<0 0 0>, rot=<0 0 0 1>}
 dyn_skirt_midl_02 {parent=dyn_skirt_midl_01, loc=<9.7844 0 0>, rot=<0 0 0 1>}
 dyn_skirt_midl_03 {parent=dyn_skirt_midl_02, loc=<11.1245 0 0>, rot=<0 0 0 1>}
 dyn_skirt_midl_04 {parent=dyn_skirt_midl_03, loc=<10.1286 0 0>, rot=<0 0 0 1>}
 tassle_r_master {parent=pelvis, loc=<4.9338 -12.1138 4.0425>, rot=<0.0223 0.097 0.9771 -0.1883>}
 tassle_r_01 {parent=tassle_r_master, loc=<0 0 0>, rot=<0 0 0 1>}
 tassle_r_02 {parent=tassle_r_01, loc=<15.3817 0 0>, rot=<0 0 0 1>}
 tassle_r_03 {parent=tassle_r_02, loc=<15.3816 0 0>, rot=<0 0 0 1>}
 tassle_r_04 {parent=tassle_r_03, loc=<13.1843 0 0>, rot=<0.0056 -0.0025 0.0004 -1>}
 tassle_r_05 {parent=tassle_r_04, loc=<8.7827 0 0>, rot=<0 0 0 1>}
 tassle_l_master {parent=pelvis, loc=<3.6752 -12.6481 2.7028>, rot=<-0.0009 -0.0038 0.9827 -0.1849>}
 tassle_l_01 {parent=tassle_l_master, loc=<0 0 0>, rot=<0 0 0 1>}
 tassle_l_02 {parent=tassle_l_01, loc=<13.1843 0 0>, rot=<0 0 0 1>}
 tassle_l_03 {parent=tassle_l_02, loc=<15.3816 0 0>, rot=<0 0 0 1>}
 tassle_l_04 {parent=tassle_l_03, loc=<13.1844 0 0>, rot=<0 0 0 1>}
 tassle_l_05 {parent=tassle_l_04, loc=<10.9868 0 0>, rot=<0 0 0 1>}
 dyn_skirt_fl_base {parent=pelvis, loc=<-2.5693 -0.0109 -7.7501>, rot=<-0.0918 0.1745 0.9798 0.0343>}
 dyn_skirt_fl_master {parent=dyn_skirt_fl_base, loc=<-5.8402 -8.6361 -0.0025>, rot=<-0.0872 0 0 -0.9962>}
 dyn_skirt_fl_01 {parent=dyn_skirt_fl_master, loc=<0 0 0>, rot=<0.0054 0.0145 0.0634 -0.9979>}
 dyn_skirt_fl_02 {parent=dyn_skirt_fl_01, loc=<7.6389 0 0>, rot=<0 0 0 1>}
 dyn_skirt_fl_03 {parent=dyn_skirt_fl_02, loc=<9.2509 0 0>, rot=<0 0 0 1>}
 dyn_skirt_fl_04 {parent=dyn_skirt_fl_03, loc=<8.3994 0 0>, rot=<0 0 0 1>}
 dyn_skirt_fr_base {parent=pelvis, loc=<-2.5691 -0.0109 7.7501>, rot=<0.0804 -0.174 0.9805 0.0435>}
 dyn_skirt_fr_master {parent=dyn_skirt_fr_base, loc=<-5.5746 -8.6156 -0.177>, rot=<0.0872 0 0 -0.9962>}
 dyn_skirt_fr_01 {parent=dyn_skirt_fr_master, loc=<0 0 0>, rot=<0 0 0 1>}
 dyn_skirt_fr_02 {parent=dyn_skirt_fr_01, loc=<7.639 0 0>, rot=<0 0 0 1>}
 dyn_skirt_fr_03 {parent=dyn_skirt_fr_02, loc=<9.2509 0.0001 0>, rot=<0 0 0 1>}
 dyn_skirt_fr_04 {parent=dyn_skirt_fr_03, loc=<8.3994 0 0>, rot=<0 0 0 1>}
 ik_foot_root {parent=root, loc=<0 0 0>, rot=<0 0 0 1>}
 ik_foot_l {parent=ik_foot_root, loc=<15.2608 -4.4176 9.4433>, rot=<0.0357 0.7045 0.0284 -0.7082>}
 ik_foot_r {parent=ik_foot_root, loc=<-15.2609 -4.4176 9.4431>, rot=<0.7082 0.0284 -0.7045 0.0357>}
 ik_hand_root {parent=root, loc=<0 0 0>, rot=<0 0 0 1>}
 ik_hand_gun {parent=ik_hand_root, loc=<-51.9729 4.0884 102.6935>, rot=<0.6084 0.1793 0.4613 -0.6205>}
 ik_hand_l {parent=ik_hand_gun, loc=<53.0286 82.1782 35.2051>, rot=<-0.5102 -0.7906 -0.3387 0>}
 ik_hand_r {parent=ik_hand_gun, loc=<0 0 0>, rot=<0 0 0 1>}
]
variables [CrouchingAlpha {type=double, default=0}]
