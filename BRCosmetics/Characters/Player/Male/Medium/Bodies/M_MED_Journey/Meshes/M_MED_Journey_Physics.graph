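# Physics asset for the skirt on the M_MED_Journey male medium cosmetic body, built from capsules on dynamic skirt bone chains and both thighs.
# Tight angular limits, mostly with zero twist, link each chain segment; six chains start at tiny master bodies, while the front two start at their first segment.
# Angularly free cross-constraints tie the back chains to their mid neighbors, covering four segments on the right but only two on the left.
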PhysicsAsset /BRCosmetics/Characters/Player/Male/Medium/Bodies/M_MED_Journey/Meshes/M_MED_Journey_Physics
bodies {
 dyn_skirt_bl_01 [{shapeType=Capsule, length=6, radius=3, shapeName=dyn_skirt_bl_01_capsule, center=<2 0 0>}]
 dyn_skirt_bl_02 [{shapeType=Capsule, length=6, radius=4.0924, shapeName=dyn_skirt_bl_02_capsule, center=<2 0 0>}]
 dyn_skirt_bl_03 [{shapeType=Capsule, length=10.0447, radius=2.9439, shapeName=dyn_skirt_bl_03_capsule, center=<2 0 0>}]
 dyn_skirt_bl_04 [{shapeType=Capsule, length=10.0447, radius=2.9439, shapeName=dyn_skirt_bl_04_capsule, center=<2 0 0>}]
 dyn_skirt_bl_05 [{shapeType=Capsule, length=10.0447, radius=2.9439, shapeName=dyn_skirt_bl_05_capsule, center=<2 0 0>}]
 dyn_skirt_bl_master [{shapeType=Capsule, length=0.505, radius=0.505, shapeName=dyn_skirt_bl_master_capsule}]
 dyn_skirt_br_01 [{shapeType=Capsule, length=6, radius=3, shapeName=dyn_skirt_br_01_capsule, center=<2 0 0>}]
 dyn_skirt_br_02 [{shapeType=Capsule, length=6, radius=4.0924, shapeName=dyn_skirt_br_02_capsule, center=<2 0 0>}]
 dyn_skirt_br_03 [{shapeType=Capsule, length=10.0447, radius=2.9439, shapeName=dyn_skirt_br_03_capsule, center=<2 0 0>}]
 dyn_skirt_br_04 [{shapeType=Capsule, length=10.0447, radius=2.9439, shapeName=dyn_skirt_br_04_capsule, center=<2 0 0>, rotation=<rot 0 0.0007 0>}]
 dyn_skirt_br_05 [{shapeType=Capsule, length=10.0447, radius=2.9439, shapeName=dyn_skirt_br_05_capsule, center=<2 0 0>}]
 dyn_skirt_br_master [{shapeType=Capsule, length=0.505, radius=0.505, shapeName=dyn_skirt_br_master_capsule}]
 dyn_skirt_fl_01 [{shapeType=Capsule, length=8, radius=3, shapeName=dyn_skirt_fl_01_capsule, center=<2 0 0>}]
 dyn_skirt_fl_02 [{shapeType=Capsule, length=11.7225, radius=3, shapeName=dyn_skirt_fl_02_capsule, center=<2 0 0>, rotation=<rot 0 0 5.0007>}]
 dyn_skirt_fl_03 [{shapeType=Capsule, length=14.5456, radius=2.4266, shapeName=dyn_skirt_fl_03_capsule, center=<0.5 0 0>}]
 dyn_skirt_fl_04 [{shapeType=Capsule, length=11.6401, radius=2.0725, shapeName=dyn_skirt_fl_04_capsule, center=<0.5 0 0>, rotation=<rot 0 0 5>}]
 dyn_skirt_fr_01 [{shapeType=Capsule, length=8, radius=3, shapeName=dyn_skirt_fr_01_capsule, center=<2 0 0>}]
 dyn_skirt_fr_02 [{shapeType=Capsule, length=11.7225, radius=3, shapeName=dyn_skirt_fl_02_capsule, center=<2 -0.0879 -1.0045>, rotation=<rot 0 0 5.0007>}]
 dyn_skirt_fr_03 [{shapeType=Capsule, length=14.5456, radius=2.4266, shapeName=dyn_skirt_fl_03_capsule, center=<0.5 0 -1.3116>}]
 dyn_skirt_fr_04 [{shapeType=Capsule, length=11.6401, radius=2.0725, shapeName=dyn_skirt_fl_04_capsule, center=<0.5 -0.1138 -1.3009>, rotation=<rot 0 0 5>}]
 dyn_skirt_l_01 [{shapeType=Capsule, length=7.5557, radius=3.4443, shapeName=dyn_skirt_l_01_capsule, center=<2 0.3798 0>}]
 dyn_skirt_l_02 [{shapeType=Capsule, length=8, radius=3, shapeName=dyn_skirt_l_02_capsule, center=<2 0.3798 0>}]
 dyn_skirt_l_03 [{shapeType=Capsule, length=8.4848, radius=2.5152, shapeName=dyn_skirt_l_03_capsule, center=<0.5 0.5 0>}]
 dyn_skirt_l_04 [{shapeType=Capsule, length=10.2268, radius=2.9922, shapeName=dyn_skirt_l_04_capsule, center=<0.5 0 0>}]
 dyn_skirt_l_master [{shapeType=Capsule, length=0.505, radius=0.505, shapeName=dyn_skirt_l_master_capsule}]
 dyn_skirt_midl_01 [{shapeType=Capsule, length=8, radius=3, shapeName=dyn_skirt_midl_01_capsule, center=<2 0.3314 0.3195>}]
 dyn_skirt_midl_02 [{shapeType=Capsule, length=8, radius=3, shapeName=dyn_skirt_midl_02_capsule, center=<2 0.3314 0.3195>}]
 dyn_skirt_midl_03 [{shapeType=Capsule, length=12.3066, radius=3, shapeName=dyn_skirt_midl_03_capsule, center=<2 0.3314 0.3195>}]
 dyn_skirt_midl_04 [{shapeType=Capsule, length=12.3066, radius=3, shapeName=dyn_skirt_midl_04_capsule, center=<2 0.3314 0.3195>}]
 dyn_skirt_midl_master [{shapeType=Capsule, length=0.505, radius=0.505, shapeName=dyn_skirt_midl_master_capsule}]
 dyn_skirt_midr_01 [{shapeType=Capsule, length=8, radius=3, shapeName=dyn_skirt_midr_01_capsule, center=<2 0 0>}]
 dyn_skirt_midr_02 [{shapeType=Capsule, length=8, radius=3, shapeName=dyn_skirt_midr_02_capsule, center=<2 0 0>}]
 dyn_skirt_midr_03 [{shapeType=Capsule, length=12.3066, radius=3, shapeName=dyn_skirt_midr_03_capsule, center=<2 0 0>}]
 dyn_skirt_midr_04 [{shapeType=Capsule, length=12.3066, radius=3, shapeName=dyn_skirt_midr_04_capsule, center=<2 0 0>}]
 dyn_skirt_midr_master [{shapeType=Capsule, length=0.505, radius=0.505, shapeName=dyn_skirt_midr_master_capsule}]
 dyn_skirt_r_01 [{shapeType=Capsule, length=8, radius=3, shapeName=dyn_skirt_r_01_capsule, center=<2.5 0.66 0>}]
 dyn_skirt_r_02 [{shapeType=Capsule, length=8, radius=3, shapeName=dyn_skirt_r_02_capsule, center=<2.5 1.5 0>}]
 dyn_skirt_r_03 [{shapeType=Capsule, length=11.4549, radius=3, shapeName=dyn_skirt_r_03_capsule, center=<2.5 1.5 0>}]
 dyn_skirt_r_04 [{shapeType=Capsule, length=10.6052, radius=3, shapeName=dyn_skirt_r_04_capsule, center=<2.5 1.5 0>}]
 dyn_skirt_r_master [{shapeType=Capsule, length=0.505, radius=0.505, shapeName=dyn_skirt_r_master_capsule}]
 thigh_l [{shapeType=Capsule, length=26.4204, radius=8.3118, shapeName=thigh_r_capsule, center=<-25 0 0>, rotation=<rot 90 0 0>}]
 thigh_r [{shapeType=Capsule, length=26.6704, radius=8.0618, shapeName=thigh_r_capsule, center=<25 0 0>, rotation=<rot 90 0 0>}]
}
constraints [
 {bone1Name=dyn_skirt_r_01, bone2Name=dyn_skirt_r_master, swing1LimitDegrees=10, swing2LimitDegrees=5, twistLimitDegrees=0}
 {bone1Name=dyn_skirt_r_02, bone2Name=dyn_skirt_r_01, swing1LimitDegrees=10, swing2LimitDegrees=5, twistLimitDegrees=0}
 {bone1Name=dyn_skirt_r_03, bone2Name=dyn_skirt_r_02, swing1LimitDegrees=5, swing2LimitDegrees=5, twistLimitDegrees=0}
 {bone1Name=dyn_skirt_r_04, bone2Name=dyn_skirt_r_03, swing1LimitDegrees=5, swing2LimitDegrees=5, twistLimitDegrees=0}
 {bone1Name=dyn_skirt_midr_01, bone2Name=dyn_skirt_midr_master, swing1LimitDegrees=5, swing2LimitDegrees=5, twistLimitDegrees=0}
 {bone1Name=dyn_skirt_midr_02, bone2Name=dyn_skirt_midr_01, swing1LimitDegrees=5, swing2LimitDegrees=5, twistLimitDegrees=0}
 {bone1Name=dyn_skirt_midr_03, bone2Name=dyn_skirt_midr_02, swing1LimitDegrees=5, swing2LimitDegrees=5, twistLimitDegrees=0}
 {bone1Name=dyn_skirt_midr_04, bone2Name=dyn_skirt_midr_03, swing1LimitDegrees=5, swing2LimitDegrees=5, twistLimitDegrees=0}
 {bone1Name=dyn_skirt_br_01, bone2Name=dyn_skirt_br_master, swing1LimitDegrees=5, swing2LimitDegrees=5, twistLimitDegrees=0}
 {bone1Name=dyn_skirt_br_02, bone2Name=dyn_skirt_br_01, swing1LimitDegrees=5, swing2LimitDegrees=5, twistLimitDegrees=0}
 {bone1Name=dyn_skirt_br_03, bone2Name=dyn_skirt_br_02, swing1LimitDegrees=7, swing2LimitDegrees=5, twistLimitDegrees=0}
 {bone1Name=dyn_skirt_br_04, bone2Name=dyn_skirt_br_03, swing1LimitDegrees=7, swing2LimitDegrees=5, twistLimitDegrees=0}
 {bone1Name=dyn_skirt_br_05, bone2Name=dyn_skirt_br_04, swing1LimitDegrees=7, swing2LimitDegrees=5, twistLimitDegrees=0}
 {bone1Name=dyn_skirt_bl_01, bone2Name=dyn_skirt_bl_master, swing1LimitDegrees=5, swing2LimitDegrees=5, twistLimitDegrees=0}
 {bone1Name=dyn_skirt_bl_02, bone2Name=dyn_skirt_bl_01, swing1LimitDegrees=5, swing2LimitDegrees=5, twistLimitDegrees=0}
 {bone1Name=dyn_skirt_bl_03, bone2Name=dyn_skirt_bl_02, swing1LimitDegrees=7, swing2LimitDegrees=5, twistLimitDegrees=0}
 {bone1Name=dyn_skirt_bl_04, bone2Name=dyn_skirt_bl_03, swing1LimitDegrees=7, swing2LimitDegrees=5, twistLimitDegrees=0}
 {bone1Name=dyn_skirt_bl_05, bone2Name=dyn_skirt_bl_04, swing1LimitDegrees=7, swing2LimitDegrees=5, twistLimitDegrees=0}
 {bone1Name=dyn_skirt_l_01, bone2Name=dyn_skirt_l_master, swing1LimitDegrees=5, swing2LimitDegrees=5, twistLimitDegrees=5}
 {bone1Name=dyn_skirt_l_02, bone2Name=dyn_skirt_l_01, swing1LimitDegrees=10, swing2LimitDegrees=5, twistLimitDegrees=0}
 {bone1Name=dyn_skirt_l_03, bone2Name=dyn_skirt_l_02, swing1LimitDegrees=10, swing2LimitDegrees=5, twistLimitDegrees=0}
 {bone1Name=dyn_skirt_l_04, bone2Name=dyn_skirt_l_03, swing1LimitDegrees=10, swing2LimitDegrees=5, twistLimitDegrees=0}
 {bone1Name=dyn_skirt_midl_01, bone2Name=dyn_skirt_midl_master, swing1LimitDegrees=2, swing2LimitDegrees=5, twistLimitDegrees=0}
 {bone1Name=dyn_skirt_midl_02, bone2Name=dyn_skirt_midl_01, swing1LimitDegrees=2, swing2LimitDegrees=5, twistLimitDegrees=0}
 {bone1Name=dyn_skirt_midl_03, bone2Name=dyn_skirt_midl_02, swing1LimitDegrees=5, swing2LimitDegrees=5, twistLimitDegrees=0}
 {bone1Name=dyn_skirt_midl_04, bone2Name=dyn_skirt_midl_03, swing1LimitDegrees=5, swing2LimitDegrees=5, twistLimitDegrees=0}
 {
  bone1Name=dyn_skirt_midr_01
  bone2Name=dyn_skirt_br_01
  swing1LimitDegrees=45
  swing1Motion=Free
  swing2LimitDegrees=45
  swing2Motion=Free
  twistLimitDegrees=45
  twistMotion=Free
 }
 {
  bone1Name=dyn_skirt_midr_02
  bone2Name=dyn_skirt_br_02
  swing1LimitDegrees=45
  swing1Motion=Free
  swing2LimitDegrees=45
  swing2Motion=Free
  twistLimitDegrees=45
  twistMotion=Free
 }
 {
  bone1Name=dyn_skirt_midr_03
  bone2Name=dyn_skirt_br_03
  swing1LimitDegrees=45
  swing1Motion=Free
  swing2LimitDegrees=45
  swing2Motion=Free
  twistLimitDegrees=45
  twistMotion=Free
 }
 {
  bone1Name=dyn_skirt_midr_04
  bone2Name=dyn_skirt_br_04
  swing1LimitDegrees=45
  swing1Motion=Free
  swing2LimitDegrees=45
  swing2Motion=Free
  twistLimitDegrees=45
  twistMotion=Free
 }
 {
  bone1Name=dyn_skirt_bl_02
  bone2Name=dyn_skirt_midl_02
  swing1LimitDegrees=45
  swing1Motion=Free
  swing2LimitDegrees=45
  swing2Motion=Free
  twistLimitDegrees=45
  twistMotion=Free
 }
 {
  bone1Name=dyn_skirt_bl_03
  bone2Name=dyn_skirt_midl_03
  swing1LimitDegrees=45
  swing1Motion=Free
  swing2LimitDegrees=45
  swing2Motion=Free
  twistLimitDegrees=45
  twistMotion=Free
 }
 {bone1Name=dyn_skirt_fl_02, bone2Name=dyn_skirt_fl_01, swing1LimitDegrees=5, swing2LimitDegrees=5, twistLimitDegrees=5}
 {bone1Name=dyn_skirt_fl_03, bone2Name=dyn_skirt_fl_02, swing1LimitDegrees=10, swing2LimitDegrees=5, twistLimitDegrees=5}
 {bone1Name=dyn_skirt_fl_04, bone2Name=dyn_skirt_fl_03, swing1LimitDegrees=15, swing2LimitDegrees=15, twistLimitDegrees=5}
 {bone1Name=dyn_skirt_fr_02, bone2Name=dyn_skirt_fr_01, swing1LimitDegrees=5, swing2LimitDegrees=5, twistLimitDegrees=5}
 {bone1Name=dyn_skirt_fr_03, bone2Name=dyn_skirt_fr_02, swing1LimitDegrees=5, swing2LimitDegrees=5, twistLimitDegrees=5}
 {bone1Name=dyn_skirt_fr_04, bone2Name=dyn_skirt_fr_03, swing1LimitDegrees=5, swing2LimitDegrees=5, twistLimitDegrees=5}
]
