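# Physics asset for the Journey male medium body cosmetic that defines simulated bodies for its tassels and back skirt bones.
# Capsules cover left and right five-segment tassel chains, each hanging from a tiny master capsule, plus two three-segment skirt chains.
# Tassel joints use tight swing and twist limits, while skirt joints set all angular motion to Free.
# No constraints link the separate chains to one another.
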PhysicsAsset /BRCosmetics/Characters/Player/Male/Medium/Bodies/M_MED_Journey/Meshes/M_MED_Journey_Tassle_Physics
bodies {
 dyn_skirt_bl_03 [{shapeType=Capsule, length=8, radius=4, shapeName=dyn_skirt_bl_03_capsule, center=<0 -2.2511 0>}]
 dyn_skirt_bl_04 [{shapeType=Capsule, length=8, radius=4, shapeName=dyn_skirt_bl_04_capsule, center=<0 -2.2511 0>}]
 dyn_skirt_bl_05 [{shapeType=Capsule, length=8, radius=4, shapeName=dyn_skirt_bl_05_capsule, center=<0 -2.2511 0>}]
 dyn_skirt_br_03 [{shapeType=Capsule, length=8, radius=4, shapeName=dyn_skirt_br_03_capsule, center=<0 -2.2511 0>}]
 dyn_skirt_br_04 [{shapeType=Capsule, length=8, radius=4, shapeName=dyn_skirt_br_04_capsule, center=<0 -2.2511 0>}]
 dyn_skirt_br_05 [{shapeType=Capsule, length=8, radius=4, shapeName=dyn_skirt_br_05_capsule, center=<0 -2.2511 0>}]
 tassle_l_01 [{shapeType=Capsule, length=6, radius=3, shapeName=tassle_l_01_capsule, center=<3 0 0>, rotation=<rot 90 0 0>}]
 tassle_l_02 [{shapeType=Capsule, length=6, radius=3, shapeName=tassle_l_02_capsule, center=<3 0 0>, rotation=<rot 90 0 0>}]
 tassle_l_03 [{shapeType=Capsule, length=6, radius=3, shapeName=tassle_l_03_capsule, center=<3 0 0>, rotation=<rot 90 0 0>}]
 tassle_l_04 [{shapeType=Capsule, length=6, radius=3, shapeName=tassle_l_04_capsule, center=<3 0 0>, rotation=<rot 90 0 0>}]
 tassle_l_05 [{shapeType=Capsule, length=6, radius=3, shapeName=tassle_l_05_capsule, center=<3 0 0>, rotation=<rot 90 0 0>}]
 tassle_l_master [{shapeType=Capsule, length=0.505, radius=0.505, shapeName=tassle_l_master_capsule, rotation=<rot 0 -90 0>}]
 tassle_r_01 [{shapeType=Capsule, length=6, radius=3, shapeName=tassle_r_01_capsule, center=<3 0 0>, rotation=<rot 90 0 0>}]
 tassle_r_02 [{shapeType=Capsule, length=6, radius=3, shapeName=tassle_r_02_capsule, center=<3 0 0>, rotation=<rot 90 0 0>}]
 tassle_r_03 [{shapeType=Capsule, length=6, radius=3, shapeName=tassle_r_03_capsule, center=<3 0 0>, rotation=<rot 90 0 0>}]
 tassle_r_04 [{shapeType=Capsule, length=6, radius=3, shapeName=tassle_r_04_capsule, center=<3 0 0>, rotation=<rot 90 0 0>}]
 tassle_r_05 [{shapeType=Capsule, length=6, radius=3, shapeName=tassle_r_05_capsule, center=<3 0 0>, rotation=<rot 90 0 0>}]
 tassle_r_master [{shapeType=Capsule, length=0.505, radius=0.505, shapeName=tassle_r_master_capsule, rotation=<rot 0 -90 0>}]
}
constraints [
 {bone1Name=tassle_r_01, bone2Name=tassle_r_master, swing1LimitDegrees=15, swing2LimitDegrees=15, twistLimitDegrees=5}
 {bone1Name=tassle_r_02, bone2Name=tassle_r_01, swing1LimitDegrees=15, swing2LimitDegrees=15, twistLimitDegrees=5}
 {bone1Name=tassle_r_03, bone2Name=tassle_r_02, swing1LimitDegrees=15, swing2LimitDegrees=15, twistLimitDegrees=5}
 {bone1Name=tassle_r_04, bone2Name=tassle_r_03, swing1LimitDegrees=15, swing2LimitDegrees=15, twistLimitDegrees=5}
 {bone1Name=tassle_r_05, bone2Name=tassle_r_04, swing1LimitDegrees=15, swing2LimitDegrees=15, twistLimitDegrees=5}
 {bone1Name=tassle_l_01, bone2Name=tassle_l_master, swing1LimitDegrees=15, swing2LimitDegrees=15, twistLimitDegrees=5}
 {bone1Name=tassle_l_02, bone2Name=tassle_l_01, swing1LimitDegrees=15, swing2LimitDegrees=15, twistLimitDegrees=5}
 {bone1Name=tassle_l_03, bone2Name=tassle_l_02, swing1LimitDegrees=15, swing2LimitDegrees=15, twistLimitDegrees=5}
 {bone1Name=tassle_l_04, bone2Name=tassle_l_03, swing1LimitDegrees=15, swing2LimitDegrees=15, twistLimitDegrees=5}
 {bone1Name=tassle_l_05, bone2Name=tassle_l_04, swing1LimitDegrees=15, swing2LimitDegrees=15, twistLimitDegrees=5}
 {
  bone1Name=dyn_skirt_br_05
  bone2Name=dyn_skirt_br_04
  swing1LimitDegrees=45
  swing1Motion=Free
  swing2LimitDegrees=45
  swing2Motion=Free
  twistLimitDegrees=45
  twistMotion=Free
 }
 {
  bone1Name=dyn_skirt_br_04
  bone2Name=dyn_skirt_br_03
  swing1LimitDegrees=45
  swing1Motion=Free
  swing2LimitDegrees=45
  swing2Motion=Free
  twistLimitDegrees=45
  twistMotion=Free
 }
 {
  bone1Name=dyn_skirt_bl_05
  bone2Name=dyn_skirt_bl_04
  swing1LimitDegrees=45
  swing1Motion=Free
  swing2LimitDegrees=45
  swing2Motion=Free
  twistLimitDegrees=45
  twistMotion=Free
 }
 {
  bone1Name=dyn_skirt_bl_04
  bone2Name=dyn_skirt_bl_03
  swing1LimitDegrees=45
  swing1Motion=Free
  swing2LimitDegrees=45
  swing2Motion=Free
  twistLimitDegrees=45
  twistMotion=Free
 }
]
