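# Post-processing AnimBlueprint for the Journey male body cosmetic that adds secondary motion to its skirt, tassels and coat flaps.
# Copies the attached parent's pose through a Control Rig, then applies thigh-driven skirt correctives, blended by emote and crouch state, into a 'preSim' cache.
# Standing, crouch and frontend branches each run RigidBody with their own physics asset and bone tweaks.
# The blended result feeds a tassel RigidBody, then AnimDynamics coat flaps.
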AnimBlueprint /BRCosmetics/Characters/Player/Male/Medium/Bodies/M_MED_Journey/Meshes/M_MED_Journey_AnimBP
target_skeleton=&/M_MED_Journey_Skeleton
graphs [
 AnimGraph {nodes [
   n1 = Get GravityOverride()
   n2 = Anim Dynamics - Bone: tassle_r_01!(ComponentPose, Alpha=1, GravityOverride:Vector=@n1) -> (Pose) {
    angularDampingOverride=0.9999
    angularSpringConstant=20
    bOverrideAngularDamping=true
    bUseGravityOverride=true
    simulationSpace=RootRelative
    alphaBoolBlend {blendOption=Linear}
    alphaScaleBiasClamp {scale=0.7}
    boundBone {boneName=tassle_r_01}
    componentAppliedLinearAccClamp=<3000 3000 3000>
    componentLinearAccScale=<0.3 0.3 0.3>
    componentLinearVelScale=<0.3 0.3 0.3>
    gravityOverride {z=-980}
    physicsBodyDefinitions [{
      sphereCollisionRadius=0
      boundBone {boneName=tassle_r_01}
      boxExtents=<10 20 20>
      constraintSetup {angularLimitsMax=<-15 15 15>, angularLimitsMin=<-15 -15 -15>, angularTarget=<1 0 0>}
      localJointOffset=<4 0 0>
     }]
    relativeSpaceBone {boneName=pelvis}
   }
   n3 = Anim Dynamics - Bone: tassle_r_02!(ComponentPose=@n2, Alpha=1, GravityOverride:Vector=@n1) -> (Pose) {
    angularDampingOverride=0.9999
    angularSpringConstant=20
    bOverrideAngularDamping=true
    bUseGravityOverride=true
    simulationSpace=RootRelative
    alphaBoolBlend {blendOption=Linear}
    alphaScaleBiasClamp {scale=0.7}
    boundBone {boneName=tassle_r_02}
    componentAppliedLinearAccClamp=<3000 3000 3000>
    componentLinearAccScale=<0.3 0.3 0.3>
    componentLinearVelScale=<0.3 0.3 0.3>
    gravityOverride {z=-980}
    physicsBodyDefinitions [{
      sphereCollisionRadius=0
      boundBone {boneName=tassle_r_02}
      boxExtents=<10 20 20>
      constraintSetup {angularLimitsMax=<-15 15 15>, angularLimitsMin=<-15 -15 -15>, angularTarget=<1 0 0>}
      localJointOffset=<4 0 0>
     }]
    relativeSpaceBone {boneName=pelvis}
   }
   n4 = Anim Dynamics - Bone: tassle_r_03!(ComponentPose=@n3, Alpha=1, GravityOverride:Vector=@n1) -> (Pose) {
    angularDampingOverride=0.9999
    angularSpringConstant=20
    bOverrideAngularDamping=true
    bUseGravityOverride=true
    bUseSphericalLimits=true
    simulationSpace=RootRelative
    alphaBoolBlend {blendOption=Linear}
    alphaScaleBiasClamp {scale=0.7}
    boundBone {boneName=tassle_r_03}
    componentAppliedLinearAccClamp=<3000 3000 3000>
    componentLinearAccScale=<0.3 0.3 0.3>
    componentLinearVelScale=<0.3 0.3 0.3>
    gravityOverride {z=-980}
    physicsBodyDefinitions [{
      sphereCollisionRadius=0
      boundBone {boneName=tassle_r_03}
      boxExtents=<10 20 20>
      constraintSetup {angularLimitsMax=<-15 15 15>, angularLimitsMin=<-15 -15 -10>, angularTarget=<1 0 0>}
      localJointOffset=<4 0 0>
     }]
    relativeSpaceBone {boneName=pelvis}
   }
   n5 = Anim Dynamics - Bone: tassle_r_04!(ComponentPose=@n4, Alpha=1, GravityOverride:Vector=@n1) -> (Pose) {
    angularDampingOverride=0.9999
    angularSpringConstant=20
    bOverrideAngularDamping=true
    bUseGravityOverride=true
    bUsePlanarLimit=false
    bUseSphericalLimits=true
    simulationSpace=RootRelative
    alphaBoolBlend {blendOption=Linear}
    alphaScaleBiasClamp {scale=0.7}
    boundBone {boneName=tassle_r_04}
    componentAppliedLinearAccClamp=<3000 3000 3000>
    componentLinearAccScale=<0.3 0.3 0.3>
    componentLinearVelScale=<0.3 0.3 0.3>
    gravityOverride {z=-980}
    physicsBodyDefinitions [{
      sphereCollisionRadius=0
      boundBone {boneName=tassle_r_04}
      boxExtents=<10 20 20>
      constraintSetup {angularLimitsMax=<-15 15 15>, angularLimitsMin=<-15 -15 -5>, angularTarget=<1 0 0>}
      localJointOffset=<4 0 0>
     }]
    relativeSpaceBone {boneName=pelvis}
   }
   n6 = Anim Dynamics - Bone: tassle_r_05!(ComponentPose=@n5, Alpha=1, GravityOverride:Vector=@n1) -> (Pose) {
    angularDampingOverride=0.9999
    angularSpringConstant=20
    bOverrideAngularDamping=true
    bUseGravityOverride=true
    bUsePlanarLimit=false
    bUseSphericalLimits=true
    simulationSpace=RootRelative
    alphaBoolBlend {blendOption=Linear}
    alphaScaleBiasClamp {scale=0.7}
    boundBone {boneName=tassle_r_05}
    componentAppliedLinearAccClamp=<3000 3000 3000>
    componentLinearAccScale=<0.3 0.3 0.3>
    componentLinearVelScale=<0.3 0.3 0.3>
    gravityOverride {z=-980}
    physicsBodyDefinitions [{
      sphereCollisionRadius=0
      boundBone {boneName=tassle_r_05}
      boxExtents=<10 20 20>
      constraintSetup {angularLimitsMax=<-15 15 15>, angularLimitsMin=<-15 -15 -5>, angularTarget=<1 0 0>}
      localJointOffset=<4 0 0>
     }]
    relativeSpaceBone {boneName=pelvis}
   }
   n7 = Anim Dynamics - Bone: tassle_l_01!(ComponentPose=@n6, Alpha=1, GravityOverride:Vector=@n1) -> (Pose) {
    angularDampingOverride=0.9999
    angularSpringConstant=20
    bOverrideAngularDamping=true
    bUseGravityOverride=true
    simulationSpace=RootRelative
    alphaBoolBlend {blendOption=Linear}
    alphaScaleBiasClamp {scale=0.7}
    boundBone {boneName=tassle_l_01}
    componentAppliedLinearAccClamp=<3000 3000 3000>
    componentLinearAccScale=<0.3 0.3 0.3>
    componentLinearVelScale=<0.3 0.3 0.3>
    gravityOverride {z=-980}
    physicsBodyDefinitions [{
      sphereCollisionRadius=0
      boundBone {boneName=tassle_l_01}
      boxExtents=<10 20 20>
      constraintSetup {angularLimitsMax=<-15 15 15>, angularLimitsMin=<-15 -15 -15>, angularTarget=<1 0 0>}
      localJointOffset=<4 0 0>
     }]
    relativeSpaceBone {boneName=pelvis}
   }
   n8 = Anim Dynamics - Bone: tassle_l_02!(ComponentPose=@n7, Alpha=1, GravityOverride:Vector=@n1) -> (Pose) {
    angularDampingOverride=0.9999
    angularSpringConstant=20
    bOverrideAngularDamping=true
    bUseGravityOverride=true
    simulationSpace=RootRelative
    alphaBoolBlend {blendOption=Linear}
    alphaScaleBiasClamp {scale=0.7}
    boundBone {boneName=tassle_l_02}
    componentAppliedLinearAccClamp=<3000 3000 3000>
    componentLinearAccScale=<0.3 0.3 0.3>
    componentLinearVelScale=<0.3 0.3 0.3>
    gravityOverride {z=-980}
    physicsBodyDefinitions [{
      sphereCollisionRadius=0
      boundBone {boneName=tassle_l_02}
      boxExtents=<10 20 20>
      constraintSetup {angularLimitsMax=<-15 15 15>, angularLimitsMin=<-15 -15 -15>, angularTarget=<1 0 0>}
      localJointOffset=<4 0 0>
     }]
    relativeSpaceBone {boneName=pelvis}
   }
   n9 = Anim Dynamics - Bone: tassle_l_03!(ComponentPose=@n8, Alpha=1, GravityOverride:Vector=@n1) -> (Pose) {
    angularDampingOverride=0.9999
    angularSpringConstant=20
    bOverrideAngularDamping=true
    bUseGravityOverride=true
    bUseSphericalLimits=true
    simulationSpace=RootRelative
    alphaBoolBlend {blendOption=Linear}
    alphaScaleBiasClamp {scale=0.7}
    boundBone {boneName=tassle_l_03}
    componentAppliedLinearAccClamp=<3000 3000 3000>
    componentLinearAccScale=<0.3 0.3 0.3>
    componentLinearVelScale=<0.3 0.3 0.3>
    gravityOverride {z=-980}
    physicsBodyDefinitions [{
      sphereCollisionRadius=0
      boundBone {boneName=tassle_l_03}
      boxExtents=<10 20 20>
      constraintSetup {angularLimitsMax=<-15 15 15>, angularLimitsMin=<-15 -15 -10>, angularTarget=<1 0 0>}
      localJointOffset=<4 0 0>
     }]
    relativeSpaceBone {boneName=pelvis}
   }
   n10 = Anim Dynamics - Bone: tassle_l_04!(ComponentPose=@n9, Alpha=1, GravityOverride:Vector=@n1) -> (Pose) {
    angularDampingOverride=0.9999
    angularSpringConstant=20
    bOverrideAngularDamping=true
    bUseGravityOverride=true
    bUsePlanarLimit=false
    bUseSphericalLimits=true
    simulationSpace=RootRelative
    alphaBoolBlend {blendOption=Linear}
    alphaScaleBiasClamp {scale=0.7}
    boundBone {boneName=tassle_l_04}
    componentAppliedLinearAccClamp=<3000 3000 3000>
    componentLinearAccScale=<0.3 0.3 0.3>
    componentLinearVelScale=<0.3 0.3 0.3>
    gravityOverride {z=-980}
    physicsBodyDefinitions [{
      sphereCollisionRadius=0
      boundBone {boneName=tassle_l_04}
      boxExtents=<10 20 20>
      constraintSetup {angularLimitsMax=<-15 15 15>, angularLimitsMin=<-15 -15 -5>, angularTarget=<1 0 0>}
      localJointOffset=<4 0 0>
     }]
    relativeSpaceBone {boneName=pelvis}
   }
   Anim Dynamics - Bone: tassle_l_05!(ComponentPose=@n10, Alpha=1, GravityOverride:Vector=@n1) -> (Pose) {
    angularDampingOverride=0.9999
    angularSpringConstant=20
    bOverrideAngularDamping=true
    bUseGravityOverride=true
    bUsePlanarLimit=false
    bUseSphericalLimits=true
    simulationSpace=RootRelative
    alphaBoolBlend {blendOption=Linear}
    alphaScaleBiasClamp {scale=0.7}
    boundBone {boneName=tassle_l_05}
    componentAppliedLinearAccClamp=<3000 3000 3000>
    componentLinearAccScale=<0.3 0.3 0.3>
    componentLinearVelScale=<0.3 0.3 0.3>
    gravityOverride {z=-980}
    physicsBodyDefinitions [{
      sphereCollisionRadius=0
      boundBone {boneName=tassle_l_05}
      boxExtents=<10 20 20>
      constraintSetup {angularLimitsMax=<-15 15 15>, angularLimitsMin=<-15 -15 -5>, angularTarget=<1 0 0>}
      localJointOffset=<4 0 0>
     }]
    relativeSpaceBone {boneName=pelvis}
   }
   n11 = Get bIsSprinting()
   n12 = "Transform (Modify) Bone - Bone: tassle_r_02"!(ComponentPose, bAlphaBoolEnabled:bool=@n11, Translation="0,0,0", Rotation="0.000,20,0.000", Scale="1,1,1") -> (Pose) {
    alphaInputType=Bool
    rotationMode=BMM_Additive
    rotationSpace=BCS_BoneSpace
    alphaBoolBlend {blendInTime=0.5, blendOption=Linear, blendOutTime=0.5}
    boneToModify {boneName=tassle_r_02}
    rotation {yaw=20}
   }
   n13 = "Transform (Modify) Bone - Bone: tassle_r_03"!(ComponentPose=@n12, bAlphaBoolEnabled:bool=@n11, Translation="0,0,0", Rotation="0.000,20,0.000", Scale="1,1,1") -> (Pose) {
    alphaInputType=Bool
    rotationMode=BMM_Additive
    rotationSpace=BCS_BoneSpace
    alphaBoolBlend {blendInTime=0.5, blendOption=Linear, blendOutTime=0.5}
    boneToModify {boneName=tassle_r_03}
    rotation {yaw=20}
   }
   n14 = "Transform (Modify) Bone - Bone: tassle_r_03"(ComponentPose=@n13, bAlphaBoolEnabled=@n11, Translation="0,0,0", Rotation="0.000,-5,0.000", Scale="1,1,1") {
    alphaInputType=Bool
    rotationMode=BMM_Additive
    rotationSpace=BCS_BoneSpace
    alphaBoolBlend {blendInTime=0.5, blendOption=Linear, blendOutTime=0.5}
    boneToModify {boneName=tassle_r_03}
    rotation {yaw=-5}
   }
   n15 = "Transform (Modify) Bone - Bone: tassle_r_03"(
    ComponentPose=@n14
    bAlphaBoolEnabled=Get sprint_not_crouch() -> (sprint_not_crouch:bool)
    Translation="0,0,0"
    Rotation="0.000,-5,0.000"
    Scale="1,1,1"
   ) {
    alphaInputType=Bool
    rotationMode=BMM_Additive
    rotationSpace=BCS_BoneSpace
    alphaBoolBlend {blendInTime=0.5, blendOption=Linear, blendOutTime=0.5}
    boneToModify {boneName=tassle_r_03}
    rotation {yaw=-5}
   }
   n16 = "Transform (Modify) Bone - Bone: tassle_l_02"!(ComponentPose=@n15, bAlphaBoolEnabled:bool=@n11, Translation="0,0,0", Rotation="0.000,20,0.000", Scale="1,1,1") -> (Pose) {
    alphaInputType=Bool
    rotationMode=BMM_Additive
    rotationSpace=BCS_BoneSpace
    alphaBoolBlend {blendInTime=0.5, blendOption=Linear, blendOutTime=0.5}
    boneToModify {boneName=tassle_l_02}
    rotation {yaw=20}
   }
   n17 = "Transform (Modify) Bone - Bone: tassle_l_03"!(ComponentPose=@n16, bAlphaBoolEnabled:bool=@n11, Translation="0,0,0", Rotation="0.000,20,0.000", Scale="1,1,1") -> (Pose) {
    alphaInputType=Bool
    rotationMode=BMM_Additive
    rotationSpace=BCS_BoneSpace
    alphaBoolBlend {blendInTime=0.5, blendOption=Linear, blendOutTime=0.5}
    boneToModify {boneName=tassle_l_03}
    rotation {yaw=20}
   }
   n18 = "Transform (Modify) Bone - Bone: tassle_l_03"(ComponentPose=@n17, bAlphaBoolEnabled=@n11, Translation="0,0,0", Rotation="0.000,-5,0.000", Scale="1,1,1") {
    alphaInputType=Bool
    rotationMode=BMM_Additive
    rotationSpace=BCS_BoneSpace
    alphaBoolBlend {blendInTime=0.5, blendOption=Linear, blendOutTime=0.5}
    boneToModify {boneName=tassle_l_03}
    rotation {yaw=-5}
   }
   n19 = "Transform (Modify) Bone - Bone: tassle_l_03"(ComponentPose=@n18, bAlphaBoolEnabled=@n11, Translation="0,0,0", Rotation="0.000,-5,0.000", Scale="1,1,1") {
    alphaInputType=Bool
    rotationMode=BMM_Additive
    rotationSpace=BCS_BoneSpace
    alphaBoolBlend {blendInTime=0.5, blendOption=Linear, blendOutTime=0.5}
    boneToModify {boneName=tassle_l_03}
    rotation {yaw=-5}
   }
   n20 = Get skydive() -> (skydive:bool)
   n21 = "Transform (Modify) Bone - Bone: tassle_r_01"!(ComponentPose=@n19, bAlphaBoolEnabled:bool=@n20, Translation="0,0,0", Rotation="0,15,0", Scale="1,1,1") -> (Pose) {
    alphaInputType=Bool
    rotationMode=BMM_Additive
    rotationSpace=BCS_BoneSpace
    alphaBoolBlend {blendInTime=0.5, blendOption=Linear, blendOutTime=0.5}
    boneToModify {boneName=tassle_r_01}
    rotation {yaw=15}
   }
   n22 = "Transform (Modify) Bone - Bone: tassle_r_02"(ComponentPose=@n21, bAlphaBoolEnabled=@n20, Translation="0,0,0", Rotation="0.000,15,0.000", Scale="1,1,1") {
    alphaInputType=Bool
    rotationMode=BMM_Additive
    rotationSpace=BCS_BoneSpace
    alphaBoolBlend {blendInTime=0.5, blendOption=Linear, blendOutTime=0.5}
    boneToModify {boneName=tassle_r_02}
    rotation {yaw=15}
   }
   n23 = "Transform (Modify) Bone - Bone: tassle_l_01"!(ComponentPose=@n22, bAlphaBoolEnabled:bool=@n20, Translation="0,0,0", Rotation="0,15,0", Scale="1,1,1") -> (Pose) {
    alphaInputType=Bool
    rotationMode=BMM_Additive
    rotationSpace=BCS_BoneSpace
    alphaBoolBlend {blendInTime=0.5, blendOption=Linear, blendOutTime=0.5}
    boneToModify {boneName=tassle_l_01}
    rotation {yaw=15}
   }
   "Transform (Modify) Bone - Bone: tassle_l_02"(ComponentPose=@n23, bAlphaBoolEnabled=@n20, Translation="0,0,0", Rotation="0.000,15,0.000", Scale="1,1,1") {
    alphaInputType=Bool
    rotationMode=BMM_Additive
    rotationSpace=BCS_BoneSpace
    alphaBoolBlend {blendInTime=0.5, blendOption=Linear, blendOutTime=0.5}
    boneToModify {boneName=tassle_l_02}
    rotation {yaw=15}
   }
   Output Pose(Result=Use cached pose 'postSim'()) {layerGroup=None}
   n24 = Control Rig(Source=Copy Pose From Mesh() {bUseAttachedParent=true}) {
    alphaBoolBlend {blendOption=Linear}
    controlRigAssetReference {blueprintRigClass=&/M_MED_Journey_CtrlRig1.M_MED_Journey_CtrlRig1_C}
   }
   n25 = Get thigh_r_alpha() -> (thigh_r_alpha:double)
   n26 = "Transform (Modify) Bone - Bone: dyn_skirt_fr_master"(ComponentPose=Local To Component(LocalPose=@n24), Alpha=@n25, Translation="-4,0,0", Rotation="0,-110.000,-0.000", Scale="1,1,1") {
    rotationMode=BMM_Additive
    rotationSpace=BCS_BoneSpace
    translationMode=BMM_Additive
    translationSpace=BCS_BoneSpace
    alphaBoolBlend {blendOption=Linear}
    alphaScaleBiasClamp {interpSpeedDecreasing=5, interpSpeedIncreasing=20}
    boneToModify {boneName=dyn_skirt_fr_master}
    rotation {yaw=-110}
    translation {x=-4}
   }
   n27 = Get thigh_l_alpha() -> (thigh_l_alpha:double)
   n28 = "Transform (Modify) Bone - Bone: dyn_skirt_fl_master"(ComponentPose=@n26, Alpha=@n27, Translation="0,4,0", Rotation="0,-110,0", Scale="1,1,1") {
    rotationMode=BMM_Additive
    rotationSpace=BCS_BoneSpace
    translationMode=BMM_Additive
    translationSpace=BCS_BoneSpace
    alphaBoolBlend {blendOption=Linear}
    alphaScaleBiasClamp {interpSpeedDecreasing=5, interpSpeedIncreasing=20}
    boneToModify {boneName=dyn_skirt_fl_master}
    rotation {yaw=-110}
    translation {y=4}
   }
   n29 = "Transform (Modify) Bone - Bone: dyn_skirt_r_master"(ComponentPose=@n28, Alpha=@n25, Translation="0,0,0", Rotation="0,30,0", Scale="1,1,1") {
    rotationMode=BMM_Additive
    rotationSpace=BCS_BoneSpace
    translationMode=BMM_Additive
    translationSpace=BCS_BoneSpace
    alphaBoolBlend {blendOption=Linear}
    alphaScaleBiasClamp {interpSpeedDecreasing=5, interpSpeedIncreasing=20}
    boneToModify {boneName=dyn_skirt_r_master}
    rotation {yaw=30}
   }
   n30 = "Transform (Modify) Bone - Bone: dyn_skirt_l_master"(ComponentPose=@n29, Alpha=@n27, Translation="0,0,0", Rotation="0,30,0", Scale="1,1,1") {
    rotationMode=BMM_Additive
    rotationSpace=BCS_BoneSpace
    translationMode=BMM_Additive
    translationSpace=BCS_BoneSpace
    alphaBoolBlend {blendOption=Linear}
    alphaScaleBiasClamp {interpSpeedDecreasing=5, interpSpeedIncreasing=20}
    boneToModify {boneName=dyn_skirt_l_master}
    rotation {yaw=30}
   }
   n31 = "Transform (Modify) Bone - Bone: dyn_skirt_fr_master"(ComponentPose=@n30, Alpha=1, Translation="0,0,0", Rotation="0.000,10,0.000", Scale="1,1,1") {
    rotationMode=BMM_Additive
    rotationSpace=BCS_BoneSpace
    translationMode=BMM_Additive
    translationSpace=BCS_BoneSpace
    alphaBoolBlend {blendOption=Linear}
    alphaScaleBiasClamp {interpSpeedDecreasing=5, interpSpeedIncreasing=20}
    boneToModify {boneName=dyn_skirt_fr_master}
    rotation {yaw=10}
   }
   n32 = "Transform (Modify) Bone - Bone: dyn_skirt_fl_master"(ComponentPose=@n31, Alpha=1, Translation="0,0,0", Rotation="0,-10,0", Scale="1,1,1") {
    rotationMode=BMM_Additive
    rotationSpace=BCS_BoneSpace
    translationMode=BMM_Additive
    translationSpace=BCS_BoneSpace
    alphaBoolBlend {blendOption=Linear}
    alphaScaleBiasClamp {interpSpeedDecreasing=5, interpSpeedIncreasing=20}
    boneToModify {boneName=dyn_skirt_fl_master}
    rotation {yaw=-10}
   }
   n33 = Component To Local(ComponentPose=@n32)
   n34 = Control Rig(Source=Copy Pose From Mesh() {bUseAttachedParent=true}) {
    alphaBoolBlend {blendOption=Linear}
    controlRigAssetReference {blueprintRigClass=&M_MED_Journey_CtrlRig1_C}
   }
   n35 = "Transform (Modify) Bone - Bone: dyn_skirt_fl_base"(
    ComponentPose=Local To Component(LocalPose=@n34)
    Alpha=Get thigh_l_alpha()
    Translation="-3,-10,0.000"
    Rotation="-20,-90,0.000"
    Scale="1,1,1"
   ) {
    rotationMode=BMM_Additive
    rotationSpace=BCS_BoneSpace
    translationMode=BMM_Additive
    translationSpace=BCS_BoneSpace
    alphaBoolBlend {blendOption=Linear}
    alphaScaleBiasClamp {interpSpeedDecreasing=5, interpSpeedIncreasing=20}
    boneToModify {boneName=dyn_skirt_fl_base}
    rotation {pitch=-20, yaw=-90}
    translation=<-3 -10>
   }
   n36 = "Transform (Modify) Bone - Bone: dyn_skirt_fr_master"(ComponentPose=@n35, Alpha=1, Translation="0,0,0", Rotation="-30,-40,0.000", Scale="1,1,1") {
    rotationMode=BMM_Additive
    rotationSpace=BCS_BoneSpace
    translationMode=BMM_Additive
    translationSpace=BCS_BoneSpace
    alphaBoolBlend {blendOption=Linear}
    alphaScaleBiasClamp {interpSpeedDecreasing=5, interpSpeedIncreasing=20}
    boneToModify {boneName=dyn_skirt_fr_master}
    rotation {pitch=-30, yaw=-40}
   }
   n37 = Get thigh_r_alpha()
   n38 = "Transform (Modify) Bone - Bone: dyn_skirt_fr_base"(ComponentPose=@n36, Alpha=@n37, Translation="0,0,0", Rotation="0.000,-25,0.000", Scale="1,1,1") {
    rotationMode=BMM_Additive
    rotationSpace=BCS_BoneSpace
    translationMode=BMM_Additive
    translationSpace=BCS_BoneSpace
    alphaBoolBlend {blendOption=Linear}
    alphaScaleBiasClamp {interpSpeedDecreasing=5, interpSpeedIncreasing=20}
    boneToModify {boneName=dyn_skirt_fr_base}
    rotation {yaw=-25}
   }
   n39 = "Transform (Modify) Bone - Bone: dyn_skirt_fr_master"(ComponentPose=@n38, Alpha=@n37, Translation="-3.851,-3.851,0.000", Rotation="0,-45,0", Scale="1,1,1") {
    rotationMode=BMM_Additive
    rotationSpace=BCS_BoneSpace
    translationMode=BMM_Additive
    translationSpace=BCS_BoneSpace
    alphaBoolBlend {blendOption=Linear}
    alphaScaleBiasClamp {interpSpeedDecreasing=5, interpSpeedIncreasing=20}
    boneToModify {boneName=dyn_skirt_fr_master}
    rotation {yaw=-45}
    translation=<-3.851 -3.851>
   }
   n40 = "Transform (Modify) Bone - Bone: dyn_skirt_r_master"(ComponentPose=@n39, Alpha=@n37, Translation="-4.512,3.996,-0.000", Rotation="0.000,35,0.000", Scale="1,1,1") {
    rotationMode=BMM_Additive
    rotationSpace=BCS_BoneSpace
    translationMode=BMM_Additive
    translationSpace=BCS_BoneSpace
    alphaBoolBlend {blendOption=Linear}
    alphaScaleBiasClamp {interpSpeedDecreasing=5, interpSpeedIncreasing=20}
    boneToModify {boneName=dyn_skirt_r_master}
    rotation {yaw=35}
    translation=<-4.512 3.996>
   }
   n41 = Component To Local(ComponentPose=@n40)
   n42 = Get bIsCrouching()
   n43 = Control Rig(Source=Copy Pose From Mesh() {bUseAttachedParent=true}) {
    alphaBoolBlend {blendOption=Linear}
    controlRigAssetReference {blueprintRigClass=&M_MED_Journey_CtrlRig1_C}
   }
   n44 = "Transform (Modify) Bone - Bone: dyn_skirt_fr_master"(ComponentPose=Local To Component(LocalPose=@n43), Alpha=Get thigh_r_alpha(), Translation="-4,0,0", Rotation="0,-110,-0.000", Scale="1,1,1") {
    rotationMode=BMM_Additive
    rotationSpace=BCS_BoneSpace
    translationMode=BMM_Additive
    translationSpace=BCS_BoneSpace
    alphaBoolBlend {blendOption=Linear}
    alphaScaleBiasClamp {interpSpeedDecreasing=5, interpSpeedIncreasing=20}
    boneToModify {boneName=dyn_skirt_fr_master}
    rotation {yaw=-110}
    translation {x=-4}
   }
   n45 = "Transform (Modify) Bone - Bone: dyn_skirt_fl_master"(ComponentPose=@n44, Alpha=Get thigh_l_alpha(), Translation="0,4,0", Rotation="0,-110,0", Scale="1,1,1") {
    rotationMode=BMM_Additive
    rotationSpace=BCS_BoneSpace
    translationMode=BMM_Additive
    translationSpace=BCS_BoneSpace
    alphaBoolBlend {blendOption=Linear}
    alphaScaleBiasClamp {interpSpeedDecreasing=5, interpSpeedIncreasing=20}
    boneToModify {boneName=dyn_skirt_fl_master}
    rotation {yaw=-110}
    translation {y=4}
   }
   n46 = "Transform (Modify) Bone - Bone: dyn_skirt_r_master"(ComponentPose=@n45, Alpha=Get thigh_r_alpha(), Translation="0,0,0", Rotation="0,40,0", Scale="1,1,1") {
    rotationMode=BMM_Additive
    rotationSpace=BCS_BoneSpace
    translationMode=BMM_Additive
    translationSpace=BCS_BoneSpace
    alphaBoolBlend {blendOption=Linear}
    alphaScaleBiasClamp {interpSpeedDecreasing=5, interpSpeedIncreasing=20}
    boneToModify {boneName=dyn_skirt_r_master}
    rotation {yaw=40}
   }
   n47 = "Transform (Modify) Bone - Bone: dyn_skirt_l_master"(ComponentPose=@n46, Alpha=Get thigh_l_alpha(), Translation="0,0,0", Rotation="0,30,0", Scale="1,1,1") {
    rotationMode=BMM_Additive
    rotationSpace=BCS_BoneSpace
    translationMode=BMM_Additive
    translationSpace=BCS_BoneSpace
    alphaBoolBlend {blendOption=Linear}
    alphaScaleBiasClamp {interpSpeedDecreasing=5, interpSpeedIncreasing=20}
    boneToModify {boneName=dyn_skirt_l_master}
    rotation {yaw=30}
   }
   Save cached pose 'preSim'(Pose=Blend Poses by bool(
     BlendPose_0=Component To Local(ComponentPose=@n47)
     BlendPose_1=Blend Poses by bool(BlendPose_0=@n41, BlendPose_1=@n33, BlendTime_0=0.5, BlendTime_1=0.5, bActiveValue=@n42) {blendType=Linear}
     BlendTime_0=0.5
     BlendTime_1=0.5
     bActiveValue=Get bIsPlayingEmote()
    ) {blendType=Linear}) {cacheName=preSim}
   n48 = Get jump_alpha() -> (jump_alpha:double)
   n49 = "Transform (Modify) Bone - Bone: dyn_skirt_fr_master"(ComponentPose=Local To Component(LocalPose=Use cached pose 'preSim'()), Alpha=@n48, Translation="-5,-5,0", Rotation="0,0,0", Scale="1,1,1") {
    rotationMode=BMM_Additive
    rotationSpace=BCS_BoneSpace
    translationMode=BMM_Additive
    translationSpace=BCS_BoneSpace
    alphaBoolBlend {blendOption=Linear}
    alphaScaleBiasClamp {interpSpeedDecreasing=20, interpSpeedIncreasing=20}
    boneToModify {boneName=dyn_skirt_fr_master}
    translation=<-5 -5>
   }
   n50 = "Transform (Modify) Bone - Bone: dyn_skirt_r_master"(ComponentPose=@n49, Alpha=@n48, Translation="-5,5,0", Rotation="0.000,20,0.000", Scale="1,1,1") {
    rotationMode=BMM_Additive
    rotationSpace=BCS_BoneSpace
    translationMode=BMM_Additive
    translationSpace=BCS_BoneSpace
    alphaBoolBlend {blendOption=Linear}
    alphaScaleBiasClamp {interpSpeedDecreasing=20, interpSpeedIncreasing=20}
    boneToModify {boneName=dyn_skirt_r_master}
    rotation {yaw=20}
    translation=<-5 5>
   }
   n51 = RigidBody(ComponentPose=@n50) {
    bOverrideWorldGravity=true
    overridePhysicsAsset=&/M_MED_Journey_Physics
    simulationSpace=BaseBoneSpace
    alphaBoolBlend {blendOption=Linear}
    baseBoneRef {boneName=pelvis}
    componentAppliedLinearAccClamp=<2000 2000 2000>
    componentLinearAccScale=<0.3 0.3 0.3>
    componentLinearVelScale=<0.3 0.3 0.3>
    overrideWorldGravity {z=-980}
    simSpaceSettings {maxAngularAcceleration=1000, maxAngularVelocity=3, maxLinearAcceleration=700, maxLinearVelocity=500, velocityScaleZ=0, worldAlpha=1}
   }
   n52 = "Transform (Modify) Bone - Bone: dyn_skirt_fr_master"(ComponentPose=@n51, Alpha=Get thigh_r_alpha(), Translation="0,0,0", Rotation="0,-10,-0.000", Scale="1,1,1") {
    rotationMode=BMM_Additive
    rotationSpace=BCS_BoneSpace
    translationMode=BMM_Additive
    translationSpace=BCS_BoneSpace
    alphaBoolBlend {blendOption=Linear}
    alphaScaleBiasClamp {interpSpeedDecreasing=5, interpSpeedIncreasing=20}
    boneToModify {boneName=dyn_skirt_fr_master}
    rotation {yaw=-10}
   }
   n53 = "Transform (Modify) Bone - Bone: dyn_skirt_fl_master"(ComponentPose=@n52, Alpha=Get thigh_l_alpha(), Translation="0,0,0", Rotation="0,-10,0", Scale="1,1,1") {
    rotationMode=BMM_Additive
    rotationSpace=BCS_BoneSpace
    translationMode=BMM_Additive
    translationSpace=BCS_BoneSpace
    alphaBoolBlend {blendOption=Linear}
    alphaScaleBiasClamp {interpSpeedDecreasing=5, interpSpeedIncreasing=20}
    boneToModify {boneName=dyn_skirt_fl_master}
    rotation {yaw=-10}
   }
   n54 = "Transform (Modify) Bone - Bone: dyn_skirt_r_master"(ComponentPose=@n53, Alpha=Get jump_alpha(), Translation="-1,2.5,0", Rotation="0.000,10,0.000", Scale="1,1,1") {
    rotationMode=BMM_Additive
    rotationSpace=BCS_BoneSpace
    translationMode=BMM_Additive
    translationSpace=BCS_BoneSpace
    alphaBoolBlend {blendOption=Linear}
    alphaScaleBiasClamp {interpSpeedDecreasing=20, interpSpeedIncreasing=20}
    boneToModify {boneName=dyn_skirt_r_master}
    rotation {yaw=10}
    translation=<-1 2.5>
   }
   n55 = "Transform (Modify) Bone - Bone: dyn_skirt_fl_master"(ComponentPose=Local To Component(LocalPose=Use cached pose 'preSim'()), Alpha=0.5, Translation="0,0,0", Rotation="0,-20,0", Scale="1,1,1") {
    rotationMode=BMM_Additive
    rotationSpace=BCS_BoneSpace
    translationMode=BMM_Additive
    translationSpace=BCS_BoneSpace
    alphaBoolBlend {blendOption=Linear}
    boneToModify {boneName=dyn_skirt_fl_master}
    rotation {yaw=-20}
   }
   n56 = "Transform (Modify) Bone - Bone: dyn_skirt_fr_master"(ComponentPose=@n55, Alpha=1, Translation="0,0,0", Rotation="0.000,20,0.000", Scale="1,1,1") {
    rotationMode=BMM_Additive
    rotationSpace=BCS_BoneSpace
    translationMode=BMM_Additive
    translationSpace=BCS_BoneSpace
    alphaBoolBlend {blendOption=Linear}
    boneToModify {boneName=dyn_skirt_fr_master}
    rotation {yaw=20}
   }
   n57 = RigidBody(ComponentPose=@n56) {
    bOverrideWorldGravity=true
    overridePhysicsAsset=&/M_MED_Journey_Crouch_Physics
    simulationSpace=BaseBoneSpace
    alphaBoolBlend {blendOption=Linear}
    baseBoneRef {boneName=pelvis}
    componentAppliedLinearAccClamp=<2000 2000 2000>
    componentLinearAccScale=<0.3 0.3 0.3>
    componentLinearVelScale=<0.3 0.3 0.3>
    overrideWorldGravity {z=-980}
    simSpaceSettings {maxAngularAcceleration=1000, maxAngularVelocity=3, maxLinearAcceleration=700, maxLinearVelocity=500, velocityScaleZ=0, worldAlpha=1}
   }
   n58 = "Transform (Modify) Bone - Bone: dyn_skirt_br_01"(ComponentPose=@n57, Alpha=1, Translation="0,0,0", Rotation="0,0,0", Scale="0.9,0.9,0.9") {
    scaleMode=BMM_Replace
    scaleSpace=BCS_BoneSpace
    alphaBoolBlend {blendInTime=0.5, blendOption=Linear, blendOutTime=0.5}
    boneToModify {boneName=dyn_skirt_br_01}
    scale=<0.9 0.9 0.9>
   }
   n59 = "Transform (Modify) Bone - Bone: dyn_skirt_br_02"(ComponentPose=@n58, Alpha=1, Translation="0,0,0", Rotation="0,0,0", Scale="0.9,0.9,0.9") {
    scaleMode=BMM_Replace
    scaleSpace=BCS_BoneSpace
    alphaBoolBlend {blendInTime=0.5, blendOption=Linear, blendOutTime=0.5}
    boneToModify {boneName=dyn_skirt_br_02}
    scale=<0.9 0.9 0.9>
   }
   n60 = "Transform (Modify) Bone - Bone: dyn_skirt_midr_01"(ComponentPose=@n59, Alpha=1, Translation="0,0,0", Rotation="0,0,0", Scale="0.8,0.8,0.8") {
    scaleMode=BMM_Replace
    scaleSpace=BCS_BoneSpace
    alphaBoolBlend {blendInTime=0.5, blendOption=Linear, blendOutTime=0.5}
    boneToModify {boneName=dyn_skirt_midr_01}
    scale=<0.8 0.8 0.8>
   }
   n61 = "Transform (Modify) Bone - Bone: dyn_skirt_midr_02"(ComponentPose=@n60, Alpha=1, Translation="0,0,0", Rotation="0,0,0", Scale="0.9,0.9,0.9") {
    scaleMode=BMM_Replace
    scaleSpace=BCS_BoneSpace
    alphaBoolBlend {blendInTime=0.5, blendOption=Linear, blendOutTime=0.5}
    boneToModify {boneName=dyn_skirt_midr_02}
    scale=<0.9 0.9 0.9>
   }
   n62 = "Transform (Modify) Bone - Bone: dyn_skirt_bl_01"(ComponentPose=@n61, Alpha=1, Translation="0,0,0", Rotation="0,0,0", Scale="0.9,0.9,0.9") {
    scaleMode=BMM_Replace
    scaleSpace=BCS_BoneSpace
    alphaBoolBlend {blendInTime=0.5, blendOption=Linear, blendOutTime=0.5}
    boneToModify {boneName=dyn_skirt_bl_01}
    scale=<0.9 0.9 0.9>
   }
   n63 = "Transform (Modify) Bone - Bone: dyn_skirt_bl_02"(ComponentPose=@n62, Alpha=1, Translation="0,0,0", Rotation="0,0,0", Scale="0.9,0.9,0.9") {
    scaleMode=BMM_Replace
    scaleSpace=BCS_BoneSpace
    alphaBoolBlend {blendInTime=0.5, blendOption=Linear, blendOutTime=0.5}
    boneToModify {boneName=dyn_skirt_bl_02}
    scale=<0.9 0.9 0.9>
   }
   n64 = "Transform (Modify) Bone - Bone: dyn_skirt_midl_01"(ComponentPose=@n63, Alpha=1, Translation="0,0,0", Rotation="0,0,0", Scale="0.9,0.9,0.9") {
    scaleMode=BMM_Replace
    scaleSpace=BCS_BoneSpace
    alphaBoolBlend {blendInTime=0.5, blendOption=Linear, blendOutTime=0.5}
    boneToModify {boneName=dyn_skirt_midl_01}
    scale=<0.9 0.9 0.9>
   }
   n65 = "Transform (Modify) Bone - Bone: dyn_skirt_midl_02"(ComponentPose=@n64, Alpha=1, Translation="0,0,0", Rotation="0,0,0", Scale="0.9,0.9,0.9") {
    scaleMode=BMM_Replace
    scaleSpace=BCS_BoneSpace
    alphaBoolBlend {blendInTime=0.5, blendOption=Linear, blendOutTime=0.5}
    boneToModify {boneName=dyn_skirt_midl_02}
    scale=<0.9 0.9 0.9>
   }
   n66 = "Transform (Modify) Bone - Bone: dyn_skirt_fl_02"(ComponentPose=@n65, Alpha=1, Translation="0,0,0", Rotation="0,10,0", Scale="1,1,1") {
    rotationMode=BMM_Additive
    rotationSpace=BCS_BoneSpace
    scaleMode=BMM_Replace
    scaleSpace=BCS_BoneSpace
    alphaBoolBlend {blendInTime=0.5, blendOption=Linear, blendOutTime=0.5}
    boneToModify {boneName=dyn_skirt_fl_02}
    rotation {yaw=10}
   }
   n67 = "Transform (Modify) Bone - Bone: dyn_skirt_fl_master"(ComponentPose=@n66, Alpha=0.5, Translation="0.000,0,-5", Rotation="-20.000,0.000,0.000", Scale="1,1,1") {
    rotationMode=BMM_Additive
    rotationSpace=BCS_BoneSpace
    translationMode=BMM_Additive
    translationSpace=BCS_BoneSpace
    alphaBoolBlend {blendOption=Linear}
    boneToModify {boneName=dyn_skirt_fl_master}
    rotation {pitch=-20}
    translation {z=-5}
   }
   n68 = "Transform (Modify) Bone - Bone: dyn_skirt_l_02"(ComponentPose=@n67, Alpha=0.5, Translation="0,0,0", Rotation="0,-30,0", Scale="1,1,1") {
    rotationMode=BMM_Additive
    rotationSpace=BCS_BoneSpace
    translationMode=BMM_Additive
    translationSpace=BCS_BoneSpace
    alphaBoolBlend {blendOption=Linear}
    boneToModify {boneName=dyn_skirt_l_02}
    rotation {yaw=-30}
   }
   n69 = "Transform (Modify) Bone - Bone: dyn_skirt_fr_master"(ComponentPose=@n68, Alpha=0.5, Translation="0,0,0", Rotation="40,0,0.000", Scale="1,1,1") {
    rotationMode=BMM_Additive
    rotationSpace=BCS_BoneSpace
    translationMode=BMM_Additive
    translationSpace=BCS_BoneSpace
    alphaBoolBlend {blendOption=Linear}
    boneToModify {boneName=dyn_skirt_fr_master}
    rotation {pitch=40}
   }
   n70 = Get thigh_r_alpha()
   n71 = "Transform (Modify) Bone - Bone: dyn_skirt_r_master"(ComponentPose=@n69, Alpha=@n70, Translation="0,0,0", Rotation="0,5,0", Scale="1,1,1") {
    rotationMode=BMM_Additive
    rotationSpace=BCS_BoneSpace
    translationMode=BMM_Additive
    translationSpace=BCS_BoneSpace
    alphaBoolBlend {blendOption=Linear}
    alphaScaleBiasClamp {bInterpResult=true}
    boneToModify {boneName=dyn_skirt_r_master}
    rotation {yaw=5}
   }
   n72 = "Transform (Modify) Bone - Bone: dyn_skirt_fr_master"(ComponentPose=@n71, Alpha=@n70, Translation="0,0,0", Rotation="0,-5,0", Scale="1,1,1") {
    rotationMode=BMM_Additive
    rotationSpace=BCS_BoneSpace
    translationMode=BMM_Additive
    translationSpace=BCS_BoneSpace
    alphaBoolBlend {blendOption=Linear}
    alphaScaleBiasClamp {bInterpResult=true}
    boneToModify {boneName=dyn_skirt_fr_master}
    rotation {yaw=-5}
   }
   n73 = Blend Poses by bool(BlendPose_0=Component To Local(ComponentPose=@n72), BlendPose_1=Component To Local(ComponentPose=@n54), bActiveValue=Get bIsCrouching()) {blendType=Linear}
   n74 = Get jump_alpha()
   n75 = "Transform (Modify) Bone - Bone: dyn_skirt_fr_master"(
    ComponentPose=Local To Component(LocalPose=Use cached pose 'preSim'())
    Alpha=@n74
    Translation="-5,-5,0"
    Rotation="0.000,-5,0.000"
    Scale="1,1,1"
   ) {
    rotationMode=BMM_Additive
    rotationSpace=BCS_BoneSpace
    translationMode=BMM_Additive
    translationSpace=BCS_BoneSpace
    alphaBoolBlend {blendOption=Linear}
    alphaScaleBiasClamp {interpSpeedDecreasing=20, interpSpeedIncreasing=20}
    boneToModify {boneName=dyn_skirt_fr_master}
    rotation {yaw=-5}
    translation=<-5 -5>
   }
   n76 = "Transform (Modify) Bone - Bone: dyn_skirt_r_master"(ComponentPose=@n75, Alpha=@n74, Translation="-5,5,0", Rotation="0.000,20,0.000", Scale="1,1,1") {
    rotationMode=BMM_Additive
    rotationSpace=BCS_BoneSpace
    translationMode=BMM_Additive
    translationSpace=BCS_BoneSpace
    alphaBoolBlend {blendOption=Linear}
    alphaScaleBiasClamp {interpSpeedDecreasing=20, interpSpeedIncreasing=20}
    boneToModify {boneName=dyn_skirt_r_master}
    rotation {yaw=20}
    translation=<-5 5>
   }
   n77 = RigidBody(ComponentPose=@n76) {
    bOverrideWorldGravity=true
    overridePhysicsAsset=&/M_MED_Journey_Frontend_Physics
    simulationSpace=BaseBoneSpace
    alphaBoolBlend {blendOption=Linear}
    baseBoneRef {boneName=pelvis}
    componentAppliedLinearAccClamp=<2000 2000 2000>
    componentLinearAccScale=<0.3 0.3 0.3>
    componentLinearVelScale=<0.3 0.3 0.3>
    overrideWorldGravity {z=-980}
    simSpaceSettings {maxAngularAcceleration=1000, maxAngularVelocity=3, maxLinearAcceleration=700, maxLinearVelocity=500, velocityScaleZ=0, worldAlpha=1}
   }
   n78 = "Transform (Modify) Bone - Bone: upperarm_l"(ComponentPose=@n77, Alpha=1, Translation="0,0,0", Rotation="0.000,7,-0.000", Scale="1,1,1") {
    rotationMode=BMM_Additive
    rotationSpace=BCS_BoneSpace
    translationMode=BMM_Additive
    translationSpace=BCS_BoneSpace
    alphaBoolBlend {blendOption=Linear}
    alphaScaleBiasClamp {interpSpeedDecreasing=5, interpSpeedIncreasing=20}
    boneToModify {boneName=upperarm_l}
    rotation {yaw=7}
   }
   n79 = "Transform (Modify) Bone - Bone: upperarm_r"(ComponentPose=@n78, Alpha=1, Translation="0,0,0", Rotation="0.000,7,-0.000", Scale="1,1,1") {
    rotationMode=BMM_Additive
    rotationSpace=BCS_BoneSpace
    translationMode=BMM_Additive
    translationSpace=BCS_BoneSpace
    alphaBoolBlend {blendOption=Linear}
    alphaScaleBiasClamp {interpSpeedDecreasing=5, interpSpeedIncreasing=20}
    boneToModify {boneName=upperarm_r}
    rotation {yaw=7}
   }
   n80 = Component To Local(ComponentPose=@n79)
   n81 = Get frontend_not_emote_no_weapon() -> (frontend_not_emote_no_weapon:bool)
   n82 = RigidBody(ComponentPose=Local To Component(LocalPose=Blend Poses by bool(BlendPose_0=@n80, BlendPose_1=@n73, BlendTime_0=0.25, bActiveValue=@n81) {blendType=Linear})) {
    bOverrideWorldGravity=true
    overridePhysicsAsset=&/M_MED_Journey_Tassle_Physics
    simulationSpace=BaseBoneSpace
    alphaBoolBlend {blendOption=Linear}
    baseBoneRef {boneName=pelvis}
    componentAppliedLinearAccClamp=<2000 2000 2000>
    componentLinearAccScale=<0.3 0.3 0.3>
    componentLinearVelScale=<0.3 0.3 0.3>
    overrideWorldGravity {z=-980}
    simSpaceSettings {maxAngularAcceleration=1000, maxAngularVelocity=3, maxLinearAcceleration=700, maxLinearVelocity=500, velocityScaleZ=0, worldAlpha=1}
   }
   n83 = Anim Dynamics - Bone: coat_flap_br_01(ComponentPose=@n82, Alpha=1, ExternalForce="0,0,0") {
    angularDampingOverride=0.9999
    angularSpringConstant=20000
    bOverrideAngularDamping=true
    bUseGravityOverride=true
    simulationSpace=RootRelative
    alphaBoolBlend {blendOption=Linear}
    boundBone {boneName=coat_flap_br_01}
    componentAppliedLinearAccClamp=<1500 1500 1500>
    componentLinearAccScale=<0.15 0.15 0.15>
    componentLinearVelScale=<0.15 0.15 0.15>
    gravityOverride {z=-980}
    physicsBodyDefinitions [{
      sphereCollisionRadius=0
      boundBone {boneName=coat_flap_br_01}
      boxExtents=<10 15 15>
      constraintSetup {angularLimitsMax=<7 3 7>, angularLimitsMin=<-7 -3 -7>, angularTarget=<1 0 0>}
      localJointOffset=<4 0 0>
     }]
   }
   n84 = Anim Dynamics - Bone: coat_flap_br_02(ComponentPose=@n83, Alpha=1, ExternalForce="0,0,0") {
    angularDampingOverride=0.9999
    angularSpringConstant=20000
    bOverrideAngularDamping=true
    bUseGravityOverride=true
    simulationSpace=RootRelative
    alphaBoolBlend {blendOption=Linear}
    boundBone {boneName=coat_flap_br_02}
    componentAppliedLinearAccClamp=<1500 1500 1500>
    componentLinearAccScale=<0.15 0.15 0.15>
    componentLinearVelScale=<0.15 0.15 0.15>
    gravityOverride {z=-980}
    physicsBodyDefinitions [{
      sphereCollisionRadius=0
      boundBone {boneName=coat_flap_br_02}
      boxExtents=<10 15 15>
      constraintSetup {angularLimitsMax=<7 3 7>, angularLimitsMin=<-7 -3 -7>, angularTarget=<1 0 0>}
      localJointOffset=<4 0 0>
     }]
   }
   n85 = Anim Dynamics - Bone: coat_flap_br_03(ComponentPose=@n84, Alpha=1, ExternalForce="0,0,0") {
    angularDampingOverride=0.9999
    angularSpringConstant=20000
    bOverrideAngularDamping=true
    bUseGravityOverride=true
    simulationSpace=RootRelative
    alphaBoolBlend {blendOption=Linear}
    boundBone {boneName=coat_flap_br_03}
    componentAppliedLinearAccClamp=<1500 1500 1500>
    componentLinearAccScale=<0.15 0.15 0.15>
    componentLinearVelScale=<0.15 0.15 0.15>
    gravityOverride {z=-980}
    physicsBodyDefinitions [{
      sphereCollisionRadius=0
      boundBone {boneName=coat_flap_br_03}
      boxExtents=<10 15 15>
      constraintSetup {angularLimitsMax=<7 3 7>, angularLimitsMin=<-7 -3 -7>, angularTarget=<1 0 0>}
      localJointOffset=<4 0 0>
     }]
   }
   n86 = Anim Dynamics - Bone: coat_flap_tr_01(ComponentPose=@n85, Alpha=1, ExternalForce="0,0,0") {
    angularDampingOverride=0.9999
    angularSpringConstant=20000
    bOverrideAngularDamping=true
    bUseGravityOverride=true
    simulationSpace=RootRelative
    alphaBoolBlend {blendOption=Linear}
    boundBone {boneName=coat_flap_tr_01}
    componentAppliedLinearAccClamp=<1500 1500 1500>
    componentLinearAccScale=<0.15 0.15 0.15>
    componentLinearVelScale=<0.15 0.15 0.15>
    gravityOverride {z=-980}
    physicsBodyDefinitions [{
      sphereCollisionRadius=0
      boundBone {boneName=coat_flap_tr_01}
      boxExtents=<10 15 15>
      constraintSetup {angularLimitsMax=<7 3 7>, angularLimitsMin=<-7 -3 -7>, angularTarget=<1 0 0>}
      localJointOffset=<4 0 0>
     }]
   }
   n87 = Anim Dynamics - Bone: coat_flap_tr_02(ComponentPose=@n86, Alpha=1, ExternalForce="0,0,0") {
    angularDampingOverride=0.9999
    angularSpringConstant=20000
    bOverrideAngularDamping=true
    bUseGravityOverride=true
    simulationSpace=RootRelative
    alphaBoolBlend {blendOption=Linear}
    boundBone {boneName=coat_flap_tr_02}
    componentAppliedLinearAccClamp=<1500 1500 1500>
    componentLinearAccScale=<0.15 0.15 0.15>
    componentLinearVelScale=<0.15 0.15 0.15>
    gravityOverride {z=-980}
    physicsBodyDefinitions [{
      sphereCollisionRadius=0
      boundBone {boneName=coat_flap_tr_02}
      boxExtents=<10 15 15>
      constraintSetup {angularLimitsMax=<7 3 7>, angularLimitsMin=<-7 -3 -7>, angularTarget=<1 0 0>}
      localJointOffset=<4 0 0>
     }]
   }
   n88 = Anim Dynamics - Bone: coat_flap_tr_03(ComponentPose=@n87, Alpha=1, ExternalForce="0,0,0") {
    angularDampingOverride=0.9999
    angularSpringConstant=20000
    bOverrideAngularDamping=true
    bUseGravityOverride=true
    simulationSpace=RootRelative
    alphaBoolBlend {blendOption=Linear}
    boundBone {boneName=coat_flap_tr_03}
    componentAppliedLinearAccClamp=<1500 1500 1500>
    componentLinearAccScale=<0.15 0.15 0.15>
    componentLinearVelScale=<0.15 0.15 0.15>
    gravityOverride {z=-980}
    physicsBodyDefinitions [{
      sphereCollisionRadius=0
      boundBone {boneName=coat_flap_tr_03}
      boxExtents=<10 15 15>
      constraintSetup {angularLimitsMax=<7 3 7>, angularLimitsMin=<-7 -3 -7>, angularTarget=<1 0 0>}
      localJointOffset=<4 0 0>
     }]
   }
   n89 = Anim Dynamics - Bone: coat_flap_bl_01(ComponentPose=@n88, Alpha=1, ExternalForce="0,0,0") {
    angularDampingOverride=0.9999
    angularSpringConstant=20000
    bOverrideAngularDamping=true
    bUseGravityOverride=true
    simulationSpace=RootRelative
    alphaBoolBlend {blendOption=Linear}
    boundBone {boneName=coat_flap_bl_01}
    componentAppliedLinearAccClamp=<1500 1500 1500>
    componentLinearAccScale=<0.15 0.15 0.15>
    componentLinearVelScale=<0.15 0.15 0.15>
    gravityOverride {z=-980}
    physicsBodyDefinitions [{
      sphereCollisionRadius=0
      boundBone {boneName=coat_flap_bl_01}
      boxExtents=<10 15 15>
      constraintSetup {angularLimitsMax=<7 3 7>, angularLimitsMin=<-7 -3 -7>, angularTarget=<1 0 0>}
      localJointOffset=<4 0 0>
     }]
   }
   n90 = Anim Dynamics - Bone: coat_flap_bl_02(ComponentPose=@n89, Alpha=1, ExternalForce="0,0,0") {
    angularDampingOverride=0.9999
    angularSpringConstant=20000
    bOverrideAngularDamping=true
    bUseGravityOverride=true
    simulationSpace=RootRelative
    alphaBoolBlend {blendOption=Linear}
    boundBone {boneName=coat_flap_bl_02}
    componentAppliedLinearAccClamp=<1500 1500 1500>
    componentLinearAccScale=<0.15 0.15 0.15>
    componentLinearVelScale=<0.15 0.15 0.15>
    gravityOverride {z=-980}
    physicsBodyDefinitions [{
      sphereCollisionRadius=0
      boundBone {boneName=coat_flap_bl_02}
      boxExtents=<10 15 15>
      constraintSetup {angularLimitsMax=<7 3 7>, angularLimitsMin=<-7 -3 -7>, angularTarget=<1 0 0>}
      localJointOffset=<4 0 0>
     }]
   }
   n91 = Anim Dynamics - Bone: coat_flap_bl_03(ComponentPose=@n90, Alpha=1, ExternalForce="0,0,0") {
    angularDampingOverride=0.9999
    angularSpringConstant=20000
    bOverrideAngularDamping=true
    bUseGravityOverride=true
    simulationSpace=RootRelative
    alphaBoolBlend {blendOption=Linear}
    boundBone {boneName=coat_flap_bl_03}
    componentAppliedLinearAccClamp=<1500 1500 1500>
    componentLinearAccScale=<0.15 0.15 0.15>
    componentLinearVelScale=<0.15 0.15 0.15>
    gravityOverride {z=-980}
    physicsBodyDefinitions [{
      sphereCollisionRadius=0
      boundBone {boneName=coat_flap_bl_03}
      boxExtents=<10 15 15>
      constraintSetup {angularLimitsMax=<7 3 7>, angularLimitsMin=<-7 -3 -7>, angularTarget=<1 0 0>}
      localJointOffset=<4 0 0>
     }]
   }
   n92 = Anim Dynamics - Bone: coat_flap_tl_01(ComponentPose=@n91, Alpha=1, ExternalForce="0,0,0") {
    angularDampingOverride=0.9999
    angularSpringConstant=20000
    bOverrideAngularDamping=true
    bUseGravityOverride=true
    simulationSpace=RootRelative
    alphaBoolBlend {blendOption=Linear}
    boundBone {boneName=coat_flap_tl_01}
    componentAppliedLinearAccClamp=<1500 1500 1500>
    componentLinearAccScale=<0.15 0.15 0.15>
    componentLinearVelScale=<0.15 0.15 0.15>
    gravityOverride {z=-980}
    physicsBodyDefinitions [{
      sphereCollisionRadius=0
      boundBone {boneName=coat_flap_tl_01}
      boxExtents=<10 15 15>
      constraintSetup {angularLimitsMax=<7 3 7>, angularLimitsMin=<-7 -3 -7>, angularTarget=<1 0 0>}
      localJointOffset=<4 0 0>
     }]
   }
   n93 = Anim Dynamics - Bone: coat_flap_tl_02(ComponentPose=@n92, Alpha=1, ExternalForce="0,0,0") {
    angularDampingOverride=0.9999
    angularSpringConstant=20000
    bOverrideAngularDamping=true
    bUseGravityOverride=true
    simulationSpace=RootRelative
    alphaBoolBlend {blendOption=Linear}
    boundBone {boneName=coat_flap_tl_02}
    componentAppliedLinearAccClamp=<1500 1500 1500>
    componentLinearAccScale=<0.15 0.15 0.15>
    componentLinearVelScale=<0.15 0.15 0.15>
    gravityOverride {z=-980}
    physicsBodyDefinitions [{
      sphereCollisionRadius=0
      boundBone {boneName=coat_flap_tl_02}
      boxExtents=<10 15 15>
      constraintSetup {angularLimitsMax=<7 3 7>, angularLimitsMin=<-7 -3 -7>, angularTarget=<1 0 0>}
      localJointOffset=<4 0 0>
     }]
   }
   n94 = Anim Dynamics - Bone: coat_flap_tr_03(ComponentPose=@n93, Alpha=1, ExternalForce="0,0,0") {
    angularDampingOverride=0.9999
    angularSpringConstant=20000
    bOverrideAngularDamping=true
    bUseGravityOverride=true
    simulationSpace=RootRelative
    alphaBoolBlend {blendOption=Linear}
    boundBone {boneName=coat_flap_tr_03}
    componentAppliedLinearAccClamp=<1500 1500 1500>
    componentLinearAccScale=<0.15 0.15 0.15>
    componentLinearVelScale=<0.15 0.15 0.15>
    gravityOverride {z=-980}
    physicsBodyDefinitions [{
      sphereCollisionRadius=0
      boundBone {boneName=coat_flap_tr_03}
      boxExtents=<10 15 15>
      constraintSetup {angularLimitsMax=<7 3 7>, angularLimitsMin=<-7 -3 -7>, angularTarget=<1 0 0>}
      localJointOffset=<4 0 0>
     }]
   }
   Save cached pose 'postSim'(Pose=Component To Local(ComponentPose=@n94)) {cacheName=postSim}
  ]}
 EventGraph {nodes [
   Event BlueprintUpdateAnimation()
   > n1 = Sequence!(execute) -> (then_0, then_1, then_2:Exec)
   > n2 = Branch(Condition=Get bClothEnabled())
   n3 = Gravity Override State Function DEPRECATED(
    execute=@n2.else
    CharacterPartAnimInstance=Self-Reference()
    GameTimeInSeconds=GetGameTimeInSeconds()
    GravityOverrideParams=Get Gravity Override Params()
   )
   > Set GravityOverride(GravityOverride=@n3)
   n4 = GetPartSkeletalMeshComponent()
   n5 = BreakRotator(InRot=BreakTransform(InTransform=GetDeltaTransformFromRefPose(self=@n4, BoneName=thigh_l, BaseName=pelvis)).Rotation)
   Set thigh_l_yaw(execute=@n1.then_1, thigh_l_yaw:Rotator=MakeRotator(Yaw=@n5.Yaw)) -> (Output_Get:Rotator)
   n6 = BreakRotator(InRot=BreakTransform(InTransform=GetDeltaTransformFromRefPose(self=@n4, BoneName=thigh_r, BaseName=pelvis)).Rotation)
   > Set thigh_r_yaw(thigh_r_yaw:Rotator=MakeRotator(Yaw=@n6.Yaw)) -> (Output_Get:Rotator)
   > Set thigh_l_alpha(thigh_l_alpha:double=MapRangeClamped(Value=@n5.Yaw, InRangeA=-20, InRangeB=-80, OutRangeA=0.15)) -> (Output_Get:double)
   > Set thigh_r_alpha(thigh_r_alpha:double=MapRangeClamped(Value=@n6.Yaw, InRangeA=-20, InRangeB=-80, OutRangeA=0.15)) -> (Output_Get:double)
   > n7 = Branch(Condition=Get jump_falling() -> (jump_falling:bool))
   n8 = BreakTransform(InTransform=GetDeltaTransformFromRefPose(self=GetOwnerSkeletalMeshComponent(), BoneName=thigh_r, BaseName=pelvis))
   > Set jump_alpha(jump_alpha:double=MapRangeClamped(Value=BreakRotator(InRot=@n8.Rotation).Yaw, InRangeA=0, InRangeB=-70)) -> (Output_Get:double)
   Set jump_alpha(execute=@n7.else, jump_alpha=0)
   Set skydive(
    execute=@n1.then_2
    skydive:bool=OR Boolean!(A=Get bIsSkydiveFloating(), B=Get bIsSkyDiving(), C:bool=Get bIsSkydiveDiveMode()) -> (ReturnValue)
   ) -> (Output_Get:bool)
   n9 = Get bIsCrouchSprinting()
   > Set sprint_not_crouch(sprint_not_crouch:bool=AND Boolean(A=NOT Boolean(A=@n9), B=Get bIsSprinting())) -> (Output_Get:bool)
   > Set jump_falling(jump_falling:bool=OR Boolean!(A=Get bIsJumping(), B=Get bIsFalling()) -> (ReturnValue)) -> (Output_Get:bool)
   n10 = Get bIsPlayingEmote()
   n11 = Get bIsPreviewingItemInFrontEnd() -> (bIsPreviewingItemInFrontEnd:bool)
   > Set frontend_not_emote_no_weapon(frontend_not_emote_no_weapon:bool=AND Boolean!(A=NOT Boolean(A=@n10), B=Get bIsInFrontEnd(), C:bool=NOT Boolean(A=@n11)) -> (ReturnValue)) -> (Output_Get:bool)
   MapRangeClamped(Value=@n6.Yaw, InRangeA=-20, InRangeB=-100, OutRangeA=0.2)
  ]}
]
variables [
 frontend_not_emote_no_weapon {type=boolean}
 Gravity Override Params {type=GravityOverrideParamsStruct, default {skydive_Parachuting {gravityOverride=<0 -800 0>}}}
 GravityOverride {type=Vector}
 jump_alpha {type=number}
 jump_falling {type=boolean}
 skydive {type=boolean}
 sprint_not_crouch {type=boolean}
 thigh_l_alpha {type=number}
 thigh_l_yaw {type=Rotator}
 thigh_r_alpha {type=number}
 thigh_r_yaw {type=Rotator}
]
